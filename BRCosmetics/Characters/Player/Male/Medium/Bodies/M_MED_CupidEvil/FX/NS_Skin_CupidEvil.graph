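# Niagara cosmetic effect for the M_MED_CupidEvil player skin, adding embers, glow and smoke-like ribbons to the skeletal mesh.
# Waveform-driven embers and an orange glow sample the mesh, while a trailing ribbon cools from red to grey-blue and fades with speed.
# A BoneFlow ribbon spirals along the right arm bones.
# A system Timeline fades most layers, user skydiving and hand-hidden flags suppress particles, and time of day tints ribbons.
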NiagaraSystem /BRCosmetics/Characters/Player/Male/Medium/Bodies/M_MED_CupidEvil/FX/NS_Skin_CupidEvil
dependencies {
 usedDataInterfaces [
  &NiagaraDataInterfaceCurve
  &NiagaraDataInterfaceColorCurve
  &NiagaraDataInterfaceVector2DCurve
  &NiagaraDataInterfaceSkeletalMesh
 ]
 usedDynamicInputs [
  &Multiply_Vector2D
  &MakeVector2D
  &MakeFloatFromVector
  &Frac_Float
 ]
 usedModules [
  &self:SystemSpawnScript.SystemScriptSource.SystemScriptGraph.NiagaraNodeAssignment_1.SetVariables_#1
  &self:SystemSpawnScript.SystemScriptSource.SystemScriptGraph.NiagaraNodeAssignment_0.SetVariables_#2
  &SystemState
  &Timeline
  /Niagara/Noise/Waveform
 ]
}
emitters {
 Embers {
  emitterUpdateScript [
   &EmitterState()
   &SpawnRate(
    Spawn Probability=&MakeFloatFromInt(Int=&MakeCustomIntFromBool(Bool=$User.IsSkydiving, True INT=0, False INT=1))
    SpawnRate=$System.Waveform.Waveform
    Use Spawn Probability=true
   )
  ]
  particleSpawnScript [
   &InitializeParticle(
    Color Mode=::NewEnumerator0
    Lifetime Max=0.9
    Lifetime Min=0.35
    Lifetime Mode=::NewEnumerator1
    Sprite Size Max=<1.7 3.5>
    Sprite Size Min=<1.2 2.5>
    Sprite Size Mode=::NewEnumerator2
    UsePositionOffset=false
    Write Material Random=true
   )
   &SkeletalMeshLocation(Bone Sampling Mode=::NewEnumerator4, MaskSpawnGroup=false, Sampled Position Offset=<-5 2 0>, Skeletal Mesh=$System.Skeletal Mesh)
   &AddVelocity(
    Module.UseConeFalloff=false
    Module.UseVelDistribution=false
    Override Seed=false
    Velocity Mode=::NewEnumerator2
    Velocity Speed=&RandomRangeFloat(Minimum=25, Maximum=100)
   )
   &InheritVelocity(Speed Limit=true, UseSpeedThreshold=false) {enabled=false}
   &ShapeLocation(MaskSpawnGroup=false, Non Uniform Scale=<0 1 1>, Override Seed=false, Sphere Radius=4, UseOwnerScale=true)
   &ApplyInitialForces()
  ]
  particleUpdateScript [
   &ParticleState()
   &Drag(Drag=2, Use Linear Drag=true, Use Rotational Drag=false)
   &CurlNoiseForce(Noise Frequency=75, Noise Strength=100, Pan Noise=true, Pan Noise Field=<0 1 1>)
   &GravityForce(Gravity=<0 0 250>)
   &InheritSourceMovement(Applied Position Scale=<0.6 0.6 0.9>, ApplyForce=false, ApplyPosition=true, ApplyPositionFalloff=false)
   &SolveForcesAndVelocity(Clamp Velocity=false, Limit Acceleration=false)
   &ScaleColor(
    Scale Alpha=&Lerp_Float(A=1, B=0, Alpha=$System.Timeline.NormalizedTime)
    Scale RGB=&MakeVectorFromLinearColorRGB(Color=&ColorFromCurve(ColorCurve=DataInterface() {propertyValues {
        AlphaCurve {keys [<at 0 1 RCIM_Cubic>]}
        BlueCurve {keys [<at 0 0 RCIM_Cubic>,<at 0.3425 0 RCIM_Cubic>]}
        GreenCurve {keys [<at 0 3.3 RCIM_Cubic>,<at 0.3425 0 RCIM_Cubic>]}
        RedCurve {keys [<at 0 22 RCIM_Cubic>,<at 0.3425 1 RCIM_Cubic>]}
       }}))
    ScaleA=true
    ScaleRGB=true
   )
   &ScaleSpriteSize(
    Non-Uniform Curve Sprite Scale=DataInterface() {propertyValues {
      XCurve {keys [<at 0 1 RCIM_Cubic>,<at 1 0 RCIM_Cubic>]}
      YCurve {keys [<at 0 1 RCIM_Cubic>, {arriveTangent=-1, interpMode=RCIM_Cubic, leaveTangent=-1, time=0.5, value=0.25}, <at 1 0 RCIM_Cubic>]}
     }}
    Scale Sprite Size Mode=::NewEnumerator3
   )
  ]
  rendererClasses [&NiagaraSpriteRendererProperties]
  renderers [{rendererClass=&NiagaraSpriteRendererProperties}]
 }
 FN_AttachedRibbon001 {
  emitterUpdateScript [
   &EmitterState()
   &FN_SpawnRibbon_v2(
    Spawn Mode=::NewEnumerator0
    SpawnRate:NiagaraFloat=&MakeCustomFloatFromBool(Bool=$User.bisHandHidden, True Float=0, False Float=30)
   )
   &FN_Integrator() {enabled=false}
  ]
  particleSpawnScript [
   /Game/Effects/Niagara/Modules/Ribbons/FN_InitializeRibbon_v4(
    Enable Skeletal Position=true
    Offset Amount=8
    Offset Axis:ENiagaraOrientationAxis=::XAxis
    Ribbon Width=8
    SkeletalMesh:NiagaraDataInterfaceSkeletalMesh=$System.Skeletal Mesh
   )
   &FN_FlipRibbonV()
   &AddVelocity(Coordinate Space=::World, Velocity=<0 0 50>)
  ]
  particleUpdateScript [
   &CurlNoiseForce(Noise Strength=100, Pan Noise=false, Pan Noise Field=<0 1 1>) {enabled=false}
   &InheritSourceMovement(
    Applied Position Scale=&Multiply_VectorByFloat(
     Vector=<-0.1 -0.1 -0.1>
     Float=&FloatFromCurve(FloatCurve=DataInterface() {propertyValues {Curve {keys [{interpMode=RCIM_Cubic, tangentMode=RCTM_User, time=0.2, value=0}, {interpMode=RCIM_Cubic, tangentMode=RCTM_User, time=1, value=1}]}}})
    )
    ApplyForce=false
    ApplyPosition=true
    ApplyPositionFalloff=false
    Source Velocity:Vector3f=&TransformVector(Vector=$Particles.SampledVelocity, Source Space=::World, Destination Space=::Local)
   ) {enabled=false}
   &InheritSourceMovement(
    Applied Position Scale=&Multiply_VectorByFloat(
     Vector=<1 1 1>
     Float=&FloatFromCurve(FloatCurve=DataInterface() {propertyValues {Curve {keys [{interpMode=RCIM_Cubic, tangentMode=RCTM_User, time=0, value=1}, {interpMode=RCIM_Cubic, tangentMode=RCTM_User, time=1, value=0}]}}})
    )
    ApplyForce=false
    ApplyPosition=true
    ApplyPositionFalloff=false
   ) {module=InheritSourceMovement001}
   &SolveForcesAndVelocity(Clamp Velocity=false, Limit Acceleration=false)
   /Game/Effects/Niagara/Modules/Ribbons/FN_UpdateRibbon_v4(Enable Skeletal Position=true, SkeletalMesh:NiagaraDataInterfaceSkeletalMesh=$System.Skeletal Mesh)
   &FN_RibbonMaterial(
    Color=&ColorFromCurve(ColorCurve=DataInterface() {propertyValues {
       AlphaCurve {keys [<at 0 1 RCIM_Cubic>]}
       BlueCurve {keys [<at 0.0066 0>,<at 0.1558 0.3212 RCIM_Cubic>]}
       GreenCurve {keys [<at 0.0066 0.0543>,<at 0.1558 0.2549 RCIM_Cubic>]}
       RedCurve {keys [<at 0.0066 0.7135>,<at 0.1558 0.1946 RCIM_Cubic>]}
      }})
    Color Pan U Speed=0.1
    Opacity Pan U Speed=1.5
    Opacity Power=0.3
    Opacity Scale U=1.5
    Scale Alpha:NiagaraFloat=&FloatFromCurve(
     FloatCurve=DataInterface() {propertyValues {Curve {keys [<at 0 0>, {interpMode=RCIM_Cubic, tangentMode=RCTM_User, time=0.2, value=1}, {interpMode=RCIM_Cubic, tangentMode=RCTM_User, time=1, value=0}]}}}
     Scale Curve=&Lerp_Float(
      A=1
      B=0
      Alpha=&RemapRange(
       Input Value To Remap=&VectorLength(Vector=$Particles.SampledVelocity)
       Input Min=20
       Input Max=200
       Output Min=0
       Output Max=1
      )
     )
    )
    Scale Color:Vector3f=&MakeVectorFromLinearColorRGB(Color=&FN_Select_Color_from_Time_Of_Day(Morning=<rgba 0.6788 0.5337 0.4254 1>, Day=<rgba 1 1 1 1>, Evening=<rgba 0.6788 0.5337 0.4254 1>, Night=<rgba 0.2462 0.2979 0.4531 1>))
   )
   &ScaleRibbonWidth(Ribbon Width Scale=&FloatFromCurve(FloatCurve=DataInterface() {propertyValues {Curve {keys [{arriveTangent=2.5273, interpMode=RCIM_Cubic, leaveTangent=2.5273, tangentMode=RCTM_User, time=0, value=0.5}, <at 0.3 1>]}}}))
   /SetVariables(Particles.Lifetime=&MakeCustomFloatFromBool(Bool=$User.IsSkydiving, True Float=0, False Float=$Particles.Lifetime)) {isSetParameters=true}
  ]
  rendererClasses [&NiagaraRibbonRendererProperties]
  renderers [
   {rendererClass=&NiagaraRibbonRendererProperties}
   {rendererClass=&NiagaraRibbonRendererProperties, rendererIndex=1}
  ]
 }
 FN_BoneFlow_BETA {
  emitterSpawnScript [
   /SetVariables(Emitter.RandomSeed:NiagaraInt32=/Niagara/DynamicInputs/Random/RandomInt(Randomness Mode:ENiagaraRandomnessMode=::NewEnumerator0, Random Seed:NiagaraInt32=0, Fixed Random Seed:NiagaraBool=false)) {isSetParameters=true}
   /SetVariables(Emitter.ParticlesPerGroup=10) {isSetParameters=true}
  ]
  emitterUpdateScript [
   /SetVariables(Emitter.NumberOfGroups=1) {isSetParameters=true}
   &EmitterState(Life Cycle Mode=::NewEnumerator1, Loop Behavior=::NewEnumerator1, Loop Duration Mode=::NewEnumerator1, UseLoopDelay=false)
   &SpawnBurst_Instantaneous(
    Spawn Count=&Multiply_Int(A=$Emitter.ParticlesPerGroup, B=$Emitter.NumberOfGroups)
    Use Spawn Probability=false
   )
  ]
  particleSpawnScript [
   /SetVariables(Particles.LoopCount:NiagaraInt32=$Emitter.LoopCount) {isSetParameters=true}
   &FN_SetSpawnGroup(
    Inclusivity=true
    Normalize Spawn Group=true
    Particles Per Group:NiagaraInt32=$Emitter.ParticlesPerGroup
    Pattern=::NewEnumerator0
    Set Ribbon ID and Ribbon Link Order=true
   )
   &InitializeParticle(Sprite Size Mode=::NewEnumerator3, UsePositionOffset=false, Write Material Random=true) {enabled=false}
   &InitializeRibbon(
    Lifetime=0.5
    Module.Write Color=true
    Module.Write Lifetime=true
    Module.Write Mass=true
    Module.Write Material Random=true
    Module.Write Position=false
    Module.Write RibbonTwist=false
    Module.Write RibbonWidth=true
    Ribbon Width=10
   )
   &FN_FlipRibbonV()
  ]
  particleUpdateScript [
   &ParticleState(Kill Particles When Lifetime Has Elapsed=false, Let Infinitely Lived Particles Die When Emitter Deactivates=true)
   /Game/Effects/Niagara/Modules/BoneFlow/FN_BoneFlow_v4/FN_BoneFlow(
    Angle:NiagaraFloat=&Add_Float(
     A=&Multiply_Float(
      A=$Particles.Normalized Spawn Group
      B=&FloatFromCurve(
       FloatCurve=DataInterface() {propertyValues {Curve {keys [<at 0 1 RCIM_Cubic>,<at 1 0.5 RCIM_Cubic>]}}}
       CurveIndex=$Particles.Normalized Spawn Group
       Scale Curve=1.5
      )
     )
     B=&Multiply_Float(A=$Particles.Age, B=0.2)
    )
    Height:NiagaraFloat=$Particles.Normalized Spawn Group
    Interpolate Mesh Sampling=false
    Radius:NiagaraFloat=&FloatFromCurve(
     FloatCurve=DataInterface() {propertyValues {Curve {keys [
         {arriveTangent=0.8198, interpMode=RCIM_Cubic, leaveTangent=0.8198, tangentMode=RCTM_User, time=0, value=0.5}
         {interpMode=RCIM_Cubic, tangentMode=RCTM_User, time=1, value=1}
        ]}}}
     CurveIndex=$Particles.Normalized Spawn Group
     Scale Curve=10
    )
    SkeletalMesh=DataInterface() {propertyValues {
      DefaultMesh=null
      ExcludeBoneName=root
      OptionalComponentClass=null
      PreviewMesh=/BRCosmetics/Characters/Player/Male/Medium/Bodies/M_MED_CupidEvil/Meshes/M_MED_CupidEvil
      SkinningMode=SkinOnTheFly
      SoftSourceActor=null
      SourceMode=Default
      UvSetIndex=0
      WholeMeshLOD=-1
      bExcludeBone=false
      bReadDeformedGeometry=true
      bRequireCurrentFrameData=true
      ComponentTags []
      FilteredBones [hand_r,lowerarm_r,upperarm_r]
      FilteredSockets []
      MeshUserParameter {parameter {name=None, typeDefHandle {registeredTypeIndex=1}, varData []}}
      SamplingRegions []
     }}
    Stabalize Twisting=true
    Write Position Attribute=true
   ) {module=FN_BoneFlow001}
   /SetVariables(Particles.RibbonFacing=$Output.FN_BoneFlow001.SplineBinormal) {isSetParameters=true}
   &FN_RibbonMaterial(
    Color=<rgba 0 1 1 9>
    Opacity Pan U Offset:NiagaraFloat=/Game/Effects/Niagara/DynamicInputs/Random/FN_SeededRandomRangeFloat(
     Minimum:NiagaraFloat=0
     Maximum:NiagaraFloat=2
     Random Seed:NiagaraInt32=&Add_Integer(A=$Particles.LoopCount, B=$Particles.SpawnGroup)
    )
    Opacity Pan U Speed:NiagaraFloat=&FN_SeededRandomRangeFloat(
     Minimum:NiagaraFloat=0.1364
     Maximum:NiagaraFloat=0.5182
     Random Seed:NiagaraInt32=&Add_Integer(A=$Particles.LoopCount, B=$Particles.SpawnGroup)
    )
    Opacity Scale U:NiagaraFloat=&FN_SeededRandomRangeFloat(
     Minimum:NiagaraFloat=0.4273
     Maximum:NiagaraFloat=1.0032
     Random Seed:NiagaraInt32=&Add_Integer(A=$Particles.LoopCount, B=$Particles.SpawnGroup)
    )
    Scale Alpha:NiagaraFloat=&FloatFromCurve(
     FloatCurve=DataInterface() {propertyValues {Curve {keys [<at -0.007 0.0112>,<at 0.2367 0.9629>,<at 0.7537 0.9995>,<at 1 0>]}}}
     Scale Curve=&FloatFromCurve(
      FloatCurve=DataInterface() {propertyValues {Curve {keys [<at 0.0007 0.022>,<at 0.2494 1.0225>,<at 0.8089 0.9748>,<at 1 0>]}}}
      CurveIndex=$Particles.Normalized Spawn Group
     )
    )
    UV Basis:NiagaraFloat=$Particles.Normalized Spawn Group
   )
   &FN_RibbonMaterial(
    Color=<rgba 0.1946 0.2549 0.3212 1>
    Color Pan U Speed=0.1
    Opacity Pan U Speed=0.7
    Opacity Power:NiagaraFloat=&FloatFromCurve(
     FloatCurve=DataInterface() {propertyValues {Curve {keys [{interpMode=RCIM_Cubic, tangentMode=RCTM_User, time=0, value=0.3}, {interpMode=RCIM_Cubic, tangentMode=RCTM_User, time=1, value=1}]}}}
     CurveIndex=$Particles.Normalized Spawn Group
    )
    Opacity Scale U=3
    Scale Alpha:NiagaraFloat=&Lerp_Float(A=1, B=0, Alpha=$System.Timeline.NormalizedTime)
    Scale Color:Vector3f=&MakeVectorFromLinearColorRGB(Color=&FN_Select_Color_from_Time_Of_Day(Morning=<rgba 0.6788 0.5337 0.4254 1>, Day=<rgba 1 1 1 1>, Evening=<rgba 0.6788 0.5337 0.4254 1>, Night=<rgba 0.167 0.202 0.3073 1>))
    UV Basis:NiagaraFloat=$Particles.Normalized Spawn Group
   ) {module=FN_RibbonMaterial001}
   &ScaleRibbonWidth(Ribbon Width Scale=&FloatFromCurve(
     FloatCurve=DataInterface() {propertyValues {Curve {keys [{interpMode=RCIM_Cubic, tangentMode=RCTM_User, time=0, value=1}, {interpMode=RCIM_Cubic, tangentMode=RCTM_User, time=1, value=0.3}]}}}
     CurveIndex=$Particles.Normalized Spawn Group
    ))
  ]
  rendererClasses [&NiagaraRibbonRendererProperties, &NiagaraSpriteRendererProperties]
  renderers [
   {rendererClass=&NiagaraRibbonRendererProperties}
   {rendererClass=&NiagaraSpriteRendererProperties, rendererIndex=1}
  ]
 }
 Glow {
  emitterUpdateScript [
   &EmitterState(Life Cycle Mode=::NewEnumerator1, Loop Behavior=::NewEnumerator1, Loop Duration Mode=::NewEnumerator1, UseLoopDelay=false)
   &SpawnBurst_Instantaneous(Use Spawn Probability=false, UseLoopCountLimit=false)
  ]
  particleSpawnScript [&InitializeParticle(
    Color=<rgba 1 0.0542 0 0.2>
    Lifetime=5
    Sprite Size Mode=::NewEnumerator3
    Uniform Sprite Size=20
    UsePositionOffset=false
    Write Material Random=true
   )]
  particleUpdateScript [
   &ParticleState(Kill Particles When Lifetime Has Elapsed=false, Let Infinitely Lived Particles Die When Emitter Deactivates=true)
   &SkeletalMeshLocation(Bone Sampling Mode=::NewEnumerator4, MaskSpawnGroup=false, Sampled Position Offset=<-5 2 0>, Skeletal Mesh=$System.Skeletal Mesh)
   &ScaleColor(Scale Alpha=&Lerp_Float(A=1, B=0, Alpha=$System.Timeline.NormalizedTime), ScaleA=true, ScaleRGB=true)
  ]
  rendererClasses [&NiagaraSpriteRendererProperties]
  renderers [{rendererClass=&NiagaraSpriteRendererProperties}]
 }
}
summary {
 systemName=NS_Skin_CupidEvil
 emitters [
  {emitterName=Embers, rendererClasses [&NiagaraSpriteRendererProperties]}
  {emitterName=Glow, rendererClasses [&NiagaraSpriteRendererProperties]}
  {emitterName=FN_AttachedRibbon001, rendererClasses [&NiagaraRibbonRendererProperties]}
  {
   emitterName=FN_BoneFlow_BETA
   rendererClasses [&NiagaraRibbonRendererProperties, &NiagaraSpriteRendererProperties]
  }
 ]
 userVariables [
  User.bIsFrontEnd {type=NiagaraBool}
  User.bIsFrontEndPreview {type=NiagaraBool}
  User.bisHandHidden {type=NiagaraBool}
  User.IsSkydiving {type=NiagaraBool}
  User.RibbonIDOffset {type=NiagaraInt32}
  User.WeakPointHitFade {type=NiagaraFloat, defaultValue=0, description=""}
 ]
}
