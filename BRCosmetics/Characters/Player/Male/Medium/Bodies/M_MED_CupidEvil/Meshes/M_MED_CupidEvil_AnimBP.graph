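# Animation blueprint for the CupidEvil male body cosmetic, animating its chains, key ring, band and back stone on top of the parent mesh's pose.
# Copied pose runs through Control Rig and weighted bone constraints into a cached input.
# Hand-tuned left and right chain deform poses blend by an alpha comparison.
# Anim Dynamics then swings ring, keys and band, swapping to spring setups while skydiving, with additive bone corrections for sitting, floating and leg motion.
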AnimBlueprint /BRCosmetics/Characters/Player/Male/Medium/Bodies/M_MED_CupidEvil/Meshes/M_MED_CupidEvil_AnimBP
target_skeleton=&/M_MED_CupidEvil_Skeleton
graphs [
 AnimGraph {nodes [
   n1 = Get Still_Sitting() -> (Still_Sitting:bool)
   n2 = Use cached pose 'RightDeformPose'()
   n3 = Use cached pose 'LeftDeformPose'()
   n4 = Get AlphaR_GreaterThen_AlphaL() -> (AlphaR_GreaterThen_AlphaL:bool)
   n5 = "Transform (Modify) Bone - Bone: back_stone"!(
    ComponentPose=Local To Component(LocalPose=Blend Poses by bool(BlendPose_0=@n2, BlendPose_1=@n3, BlendTime_0=0.15, BlendTime_1=0.15, bActiveValue=@n4) {blendType=Linear})
    Alpha=1
    Translation="-0.315,-0.556,-0.013"
    Rotation="23.853,-32.838,-2.101"
   ) -> (Pose) {
    rotationMode=BMM_Additive
    rotationSpace=BCS_BoneSpace
    translationMode=BMM_Additive
    translationSpace=BCS_BoneSpace
    alphaBoolBlend {blendOption=Linear}
    boneToModify {boneName=back_stone}
    rotation=<rot 23.853 -32.838 -2.101>
    translation=<-0.315 -0.556 -0.013>
   }
   n6 = Get is not sky diving() -> (is not sky diving:bool)
   n7 = Anim Dynamics - Bone: dyn_ring!(ComponentPose=@n5, bAlphaBoolEnabled:bool=@n6, ExternalForce="0,0,0") -> (Pose) {
    alphaInputType=Bool
    simulationSpace=World
    alphaBoolBlend {blendOption=Linear}
    boundBone {boneName=dyn_ring}
    componentAppliedLinearAccClamp=<1000 1000 1000>
    componentLinearAccScale=<0.3 0.3 0.3>
    componentLinearVelScale=<0.3 0.3 0.3>
    physicsBodyDefinitions [{
      boundBone {boneName=dyn_ring}
      boxExtents=<30 30 30>
      constraintSetup {angularLimitsMax=<0 10 0>, angularLimitsMin=<0 0 0>, angularTarget=<0 0 0>}
      localJointOffset=<2 0 1>
     }]
    relativeSpaceBone {boneName=root}
   }
   n8 = Anim Dynamics - Bone: dyn_key_03!(ComponentPose=@n7, bAlphaBoolEnabled:bool=@n6, ExternalForce="0,0,0") -> (Pose) {
    alphaInputType=Bool
    angularDampingOverride=0.99
    bOverrideAngularDamping=true
    bUsePlanarLimit=false
    simulationSpace=World
    alphaBoolBlend {blendInTime=0.1, blendOption=Linear, blendOutTime=0.1}
    boundBone {boneName=dyn_key_03}
    componentAppliedLinearAccClamp=<1000 1000 1000>
    componentLinearAccScale=<0.3 0.3 0.3>
    componentLinearVelScale=<0.3 0.3 0.3>
    physicsBodyDefinitions [{
      collisionType=CustomSphere
      sphereCollisionRadius=4
      boundBone {boneName=dyn_key_03}
      boxExtents=<50 50 50>
      constraintSetup {angularLimitsMax=<0 0 10>, angularLimitsMin=<0 -15 -20>, angularTarget=<0 0 0>}
      localJointOffset=<3 0 0>
     }]
    planarLimits [{drivingBone {boneName=thigh_l}, planeTransform {location=<0.0009 -3.3101 -4.1>, rotation=<rot 0.024 -0.0198 -170.0019>}}]
    relativeSpaceBone {boneName=root}
   }
   n9 = Anim Dynamics - Bone: dyn_key_02!(ComponentPose=@n8, bAlphaBoolEnabled:bool=@n6, ExternalForce="0,0,0") -> (Pose) {
    alphaInputType=Bool
    angularDampingOverride=0.99
    bOverrideAngularDamping=true
    bUsePlanarLimit=false
    simulationSpace=World
    alphaBoolBlend {blendInTime=0.1, blendOption=Linear, blendOutTime=0.1}
    boundBone {boneName=dyn_key_02}
    componentAppliedLinearAccClamp=<1000 1000 1000>
    componentLinearAccScale=<0.3 0.3 0.3>
    componentLinearVelScale=<0.3 0.3 0.3>
    physicsBodyDefinitions [{
      collisionType=CustomSphere
      sphereCollisionRadius=4.5
      boundBone {boneName=dyn_key_02}
      boxExtents=<45 45 45>
      constraintSetup {angularLimitsMax=<0 0 20>, angularLimitsMin=<0 -15 -10>, angularTarget=<0 0 0>}
      localJointOffset=<3.5 0 0>
     }]
    planarLimits [{drivingBone {boneName=thigh_l}, planeTransform {location=<0.0009 -3.3101 -4.1>, rotation=<rot 0.024 -0.0198 -170.0019>}}]
    relativeSpaceBone {boneName=root}
   }
   n10 = Anim Dynamics - Bone: dyn_key_01!(ComponentPose=@n9, bAlphaBoolEnabled:bool=@n6, ExternalForce="0,0,0") -> (Pose) {
    alphaInputType=Bool
    angularDampingOverride=0.99
    bOverrideAngularDamping=true
    bUsePlanarLimit=false
    simulationSpace=World
    alphaBoolBlend {blendInTime=0.1, blendOption=Linear, blendOutTime=0.1}
    boundBone {boneName=dyn_key_01}
    componentAppliedLinearAccClamp=<1000 1000 1000>
    componentLinearAccScale=<0.3 0.3 0.3>
    componentLinearVelScale=<0.3 0.3 0.3>
    physicsBodyDefinitions [{
      collisionType=CustomSphere
      sphereCollisionRadius=5
      boundBone {boneName=dyn_key_01}
      boxExtents=<40 40 40>
      constraintSetup {angularLimitsMax=<0 0 10>, angularLimitsMin=<0 -30 -10>, angularTarget=<0 0 0>}
      localJointOffset=<4 0 0>
     }]
    planarLimits [{drivingBone {boneName=thigh_l}, planeTransform {location=<0.0009 -3.3101 -4.1>, rotation=<rot 0.024 -0.0198 -170.0019>}}]
    relativeSpaceBone {boneName=root}
   }
   n11 = Look At - Bone: dyn_ring!(ComponentPose=@n10, bAlphaBoolEnabled:bool=Get bIsSkyDiving()) -> (Pose) {
    alphaInputType=Bool
    alphaBoolBlend {blendInTime=0.1, blendOption=Linear, blendOutTime=0.1}
    boneToModify {boneName=dyn_ring}
    lookAtLocation {x=-50, z=-12}
    lookAtTarget {boneReference {boneName=thigh_l}}
    lookAt_Axis {axis=<1 0>}
   }
   n12 = Get bIsSkyDiving()
   n13 = Anim Dynamics - Bone: dyn_ring(ComponentPose=@n11, bAlphaBoolEnabled=@n12, ExternalForce="0,0,0") {
    alphaInputType=Bool
    angularSpringConstant=1000
    bAngularSpring=true
    gravityScale=0
    simulationSpace=BoneRelative
    alphaBoolBlend {blendInTime=0.1, blendOption=Linear, blendOutTime=0.1}
    boundBone {boneName=dyn_ring}
    componentAppliedLinearAccClamp=<1000 1000 1000>
    componentLinearAccScale=<0.3 0.3 0.3>
    componentLinearVelScale=<0.3 0.3 0.3>
    physicsBodyDefinitions [{
      boundBone {boneName=dyn_ring}
      boxExtents=<30 30 30>
      constraintSetup {angularLimitsMax=<20 0 40>, angularLimitsMin=<-20 -30 -40>, angularTarget=<0 0 0>}
      localJointOffset=<5 0 2>
     }]
    relativeSpaceBone {boneName=root}
   }
   n14 = Anim Dynamics - Bone: dyn_key_03(ComponentPose=@n13, bAlphaBoolEnabled=@n12, ExternalForce="0,0,0") {
    alphaInputType=Bool
    angularDampingOverride=0.99
    angularSpringConstant=1000
    bAngularSpring=true
    bOverrideAngularDamping=true
    bUsePlanarLimit=false
    gravityScale=0
    simulationSpace=BoneRelative
    alphaBoolBlend {blendInTime=0.1, blendOption=Linear, blendOutTime=0.1}
    boundBone {boneName=dyn_key_03}
    componentAppliedLinearAccClamp=<1000 1000 1000>
    componentLinearAccScale=<0.3 0.3 0.3>
    componentLinearVelScale=<0.3 0.3 0.3>
    physicsBodyDefinitions [{
      collisionType=CustomSphere
      sphereCollisionRadius=4
      boundBone {boneName=dyn_key_03}
      boxExtents=<40 40 40>
      constraintSetup {angularLimitsMax=<0 0 20>, angularLimitsMin=<0 -10 -20>, angularTarget=<0 0 0>}
      localJointOffset=<2 0 0>
     }]
    planarLimits [{drivingBone {boneName=thigh_l}, planeTransform {location=<0.0009 -3.3101 -4.1>, rotation=<rot 0.024 -0.0198 -170.0019>}}]
    relativeSpaceBone {boneName=root}
   }
   n15 = Anim Dynamics - Bone: dyn_key_02(ComponentPose=@n14, bAlphaBoolEnabled=@n12, ExternalForce="0,0,0") {
    alphaInputType=Bool
    angularDampingOverride=0.99
    angularSpringConstant=1000
    bAngularSpring=true
    bOverrideAngularDamping=true
    bUsePlanarLimit=false
    gravityScale=0
    simulationSpace=BoneRelative
    alphaBoolBlend {blendInTime=0.1, blendOption=Linear, blendOutTime=0.1}
    boundBone {boneName=dyn_key_02}
    componentAppliedLinearAccClamp=<1000 1000 1000>
    componentLinearAccScale=<0.3 0.3 0.3>
    componentLinearVelScale=<0.3 0.3 0.3>
    physicsBodyDefinitions [{
      collisionType=CustomSphere
      sphereCollisionRadius=4.5
      boundBone {boneName=dyn_key_02}
      boxExtents=<35 35 35>
      constraintSetup {angularLimitsMax=<0 0 20>, angularLimitsMin=<0 -15 -20>, angularTarget=<0 0 0>}
      localJointOffset=<2.5 0 0>
     }]
    planarLimits [{drivingBone {boneName=thigh_l}, planeTransform {location=<0.0009 -3.3101 -4.1>, rotation=<rot 0.024 -0.0198 -170.0019>}}]
    relativeSpaceBone {boneName=root}
   }
   n16 = Anim Dynamics - Bone: dyn_key_01(ComponentPose=@n15, bAlphaBoolEnabled=@n12, ExternalForce="0,0,0") {
    alphaInputType=Bool
    angularDampingOverride=0.99
    angularSpringConstant=1000
    bAngularSpring=true
    bOverrideAngularDamping=true
    bUsePlanarLimit=false
    gravityScale=0
    simulationSpace=BoneRelative
    alphaBoolBlend {blendInTime=0.1, blendOption=Linear, blendOutTime=0.1}
    boundBone {boneName=dyn_key_01}
    componentAppliedLinearAccClamp=<1000 1000 1000>
    componentLinearAccScale=<0.3 0.3 0.3>
    componentLinearVelScale=<0.3 0.3 0.3>
    physicsBodyDefinitions [{
      collisionType=CustomSphere
      sphereCollisionRadius=5
      boundBone {boneName=dyn_key_01}
      boxExtents=<30 30 30>
      constraintSetup {angularLimitsMax=<0 5 10>, angularLimitsMin=<0 -15 -10>, angularTarget=<0 0 0>}
      localJointOffset=<3 0 0>
     }]
    planarLimits [{drivingBone {boneName=thigh_l}, planeTransform {location=<0.0009 -3.3101 -4.1>, rotation=<rot 0.024 -0.0198 -170.0019>}}]
    relativeSpaceBone {boneName=root}
   }
   n17 = "Transform (Modify) Bone - Bone: dyn_ring"!(ComponentPose=@n16, Alpha=Get Thigh_L_Standing_Straight() -> (Thigh_L_Standing_Straight:double), Translation="0,0,0", Rotation="5,0,-20") -> (Pose) {
    rotationMode=BMM_Additive
    rotationSpace=BCS_BoneSpace
    translationMode=BMM_Additive
    translationSpace=BCS_BoneSpace
    alphaBoolBlend {blendOption=Linear}
    boneToModify {boneName=dyn_ring}
    rotation {pitch=5, roll=-20}
   }
   n18 = "Transform (Modify) Bone - Bone: dyn_ring"!(ComponentPose=@n17, Alpha=Get L Leg going Left() -> (L Leg going Left:float), Translation="0,0,0", Rotation="48,0,0", Scale:Vector="1,1,1") -> (Pose) {
    rotationMode=BMM_Additive
    rotationSpace=BCS_BoneSpace
    translationMode=BMM_Additive
    translationSpace=BCS_BoneSpace
    alphaBoolBlend {blendOption=Linear}
    alphaScaleBiasClamp {interpSpeedDecreasing=1, interpSpeedIncreasing=1}
    boneToModify {boneName=dyn_ring}
    rotation {pitch=48}
   }
   n19 = "Transform (Modify) Bone - Bone: dyn_ring"!(ComponentPose=@n18, bAlphaBoolEnabled:bool=Get Still_Sitting(), Translation="0,0,0", Rotation="-25,0,0", Scale="1,1,1") -> (Pose) {
    alphaInputType=Bool
    rotationMode=BMM_Additive
    rotationSpace=BCS_BoneSpace
    translationMode=BMM_Additive
    translationSpace=BCS_BoneSpace
    alphaBoolBlend {blendInTime=0.3, blendOption=Linear, blendOutTime=0.3}
    boneToModify {boneName=dyn_ring}
    rotation {pitch=-25}
   }
   n20 = Anim Dynamics - Bone: dyn_band_01(ComponentPose=@n19, Alpha=1, ExternalForce="0,0,0") {
    angularDampingOverride=0.9
    bOverrideAngularDamping=true
    bUsePlanarLimit=false
    simulationSpace=World
    alphaBoolBlend {blendOption=Linear}
    boundBone {boneName=dyn_band_01}
    componentAppliedLinearAccClamp=<1000 1000 1000>
    componentLinearAccScale=<0.3 0.3 0.3>
    componentLinearVelScale=<0.3 0.3 0.3>
    physicsBodyDefinitions [{
      boundBone {boneName=dyn_band_01}
      boxExtents=<15 15 15>
      constraintSetup {angularLimitsMax=<0 10 30>, angularLimitsMin=<0 -10 -30>, angularTarget=<0 0 0>}
      localJointOffset=<1.5 0 0>
     }]
   }
   n21 = Anim Dynamics - Bone: dyn_band_02(ComponentPose=@n20, Alpha=1, ExternalForce="0,0,0") {
    angularDampingOverride=0.9
    bOverrideAngularDamping=true
    bUsePlanarLimit=false
    simulationSpace=World
    alphaBoolBlend {blendOption=Linear}
    boundBone {boneName=dyn_band_02}
    componentAppliedLinearAccClamp=<1000 1000 1000>
    componentLinearAccScale=<0.3 0.3 0.3>
    componentLinearVelScale=<0.3 0.3 0.3>
    physicsBodyDefinitions [{
      boundBone {boneName=dyn_band_02}
      boxExtents=<15 15 15>
      constraintSetup {angularLimitsMax=<0 10 30>, angularLimitsMin=<0 -10 -30>, angularTarget=<0 0 0>}
      localJointOffset=<1.5 0 0>
     }]
   }
   n22 = "Transform (Modify) Bone - Bone: chain_03_back_03"(ComponentPose=@n21, Alpha=1, Translation="0,-3,-3", Rotation="0,0,0", Scale="1,1,1") {
    rotationMode=BMM_Additive
    rotationSpace=BCS_BoneSpace
    translationMode=BMM_Additive
    translationSpace=BCS_BoneSpace
    alphaBoolBlend {blendOption=Linear}
    boneToModify {boneName=chain_03_back_03}
    translation {y=-3, z=-3}
   }
   n23 = "Transform (Modify) Bone - Bone: chain_03_back_01"(ComponentPose=@n22, Alpha=1, Translation="0,0,0", Rotation="-5.000,0.000,0.000", Scale="1,1,1") {
    rotationMode=BMM_Additive
    rotationSpace=BCS_BoneSpace
    translationMode=BMM_Additive
    translationSpace=BCS_BoneSpace
    alphaBoolBlend {blendOption=Linear}
    boneToModify {boneName=chain_03_back_01}
    rotation=<rot -5 0 0.0001>
   }
   n24 = "Transform (Modify) Bone - Bone: chain_03_back_02"(ComponentPose=@n23, Alpha=1, Translation="0,-2,-2", Rotation="-5.000,-0.000,0.000", Scale="1,1,1") {
    rotationMode=BMM_Additive
    rotationSpace=BCS_BoneSpace
    translationMode=BMM_Additive
    translationSpace=BCS_BoneSpace
    alphaBoolBlend {blendOption=Linear}
    boneToModify {boneName=chain_03_back_02}
    rotation=<rot -5.0002 0 0.0001>
    translation {y=-2, z=-2}
   }
   n25 = "Transform (Modify) Bone - Bone: chain_03_front_04"(ComponentPose=@n24, Alpha=1, Translation="0,3,-4", Rotation="-5.000,0.000,0.000", Scale="1,1,1") {
    rotationMode=BMM_Additive
    rotationSpace=BCS_BoneSpace
    translationMode=BMM_Additive
    translationSpace=BCS_BoneSpace
    alphaBoolBlend {blendOption=Linear}
    boneToModify {boneName=chain_03_front_04}
    rotation=<rot -4.9999 0 0>
    translation {y=3, z=-4}
   }
   n26 = "Transform (Modify) Bone - Bone: chain_03_front_03"(ComponentPose=@n25, Alpha=1, Translation="0,1,0", Rotation="-14.939,5.167,-1.092", Scale="1,1,1") {
    rotationMode=BMM_Additive
    rotationSpace=BCS_BoneSpace
    translationMode=BMM_Additive
    translationSpace=BCS_BoneSpace
    alphaBoolBlend {blendOption=Linear}
    boneToModify {boneName=chain_03_front_03}
    rotation=<rot -14.939 5.167 -1.0919>
    translation {y=1}
   }
   n27 = "Transform (Modify) Bone - Bone: chain_03_back_01"(ComponentPose=@n26, Alpha=1, Translation="0,0,0", Rotation="-5.000,0.000,0.000", Scale="1,1,1") {
    rotationMode=BMM_Additive
    rotationSpace=BCS_BoneSpace
    translationMode=BMM_Additive
    translationSpace=BCS_BoneSpace
    alphaBoolBlend {blendOption=Linear}
    boneToModify {boneName=chain_03_back_01}
    rotation=<rot -5 0 0.0001>
   }
   n28 = "Transform (Modify) Bone - Bone: chain_03_back_03"(ComponentPose=@n27, Alpha=1, Translation="0,0,-4", Rotation="0,0,0", Scale="1,1,1") {
    rotationMode=BMM_Additive
    rotationSpace=BCS_BoneSpace
    translationMode=BMM_Additive
    translationSpace=BCS_BoneSpace
    alphaBoolBlend {blendOption=Linear}
    boneToModify {boneName=chain_03_back_03}
    translation {z=-4}
   }
   n29 = "Transform (Modify) Bone - Bone: chain_03_back_02"(ComponentPose=@n28, Alpha=1, Translation="0,0,-2.5", Rotation="-5.000,-0.000,0.000", Scale="1,1,1") {
    rotationMode=BMM_Additive
    rotationSpace=BCS_BoneSpace
    translationMode=BMM_Additive
    translationSpace=BCS_BoneSpace
    alphaBoolBlend {blendOption=Linear}
    boneToModify {boneName=chain_03_back_02}
    rotation=<rot -5.0001 0 0.0001>
    translation {z=-2.5}
   }
   n30 = "Transform (Modify) Bone - Bone: chain_03_front_04"(ComponentPose=@n29, Alpha=1, Translation="0,0,0", Rotation="-10.000,-0.000,0.000", Scale="1,1,1") {
    rotationMode=BMM_Additive
    rotationSpace=BCS_BoneSpace
    translationMode=BMM_Additive
    translationSpace=BCS_BoneSpace
    alphaBoolBlend {blendOption=Linear}
    boneToModify {boneName=chain_03_front_04}
    rotation=<rot -9.9999 0 0>
   }
   n31 = Get bIsSkydiveFloating()
   n32 = "Transform (Modify) Bone - Bone: chain_03_front_04"!(ComponentPose=@n30, bAlphaBoolEnabled:bool=@n31, Translation="-1,2,0", Rotation="0,0,0", Scale="1,1,1") -> (Pose) {
    alphaInputType=Bool
    rotationMode=BMM_Additive
    rotationSpace=BCS_BoneSpace
    translationMode=BMM_Additive
    translationSpace=BCS_BoneSpace
    alphaBoolBlend {blendInTime=0.2, blendOption=Linear, blendOutTime=0.3}
    boneToModify {boneName=chain_03_front_04}
    translation=<-1 2>
   }
   n33 = "Transform (Modify) Bone - Bone: chain_03_front_03"!(ComponentPose=@n32, bAlphaBoolEnabled:bool=@n31, Translation="0,1,0", Rotation="-0.000,10.000,0.000", Scale="1,1,1") -> (Pose) {
    alphaInputType=Bool
    rotationMode=BMM_Additive
    rotationSpace=BCS_BoneSpace
    translationMode=BMM_Additive
    translationSpace=BCS_BoneSpace
    alphaBoolBlend {blendInTime=0.2, blendOption=Linear, blendOutTime=0.3}
    boneToModify {boneName=chain_03_front_03}
    rotation=<rot 0 10 0.0001>
    translation {y=1}
   }
   n34 = "Transform (Modify) Bone - Bone: chain_01_back_03"!(ComponentPose=@n33, bAlphaBoolEnabled:bool=@n31, Translation="0.000,-1.641,-0.000", Rotation="0,0,0", Scale="1,1,1") -> (Pose) {
    alphaInputType=Bool
    rotationMode=BMM_Additive
    rotationSpace=BCS_BoneSpace
    translationMode=BMM_Additive
    translationSpace=BCS_BoneSpace
    alphaBoolBlend {blendInTime=0.2, blendOption=Linear, blendOutTime=0.3}
    boneToModify {boneName=chain_01_back_03}
    translation {y=-1.641}
   }
   n35 = "Transform (Modify) Bone - Bone: chain_01_back_02"!(ComponentPose=@n34, bAlphaBoolEnabled:bool=@n31, Translation="-0.352,-0.532,0.000", Rotation="0.000,-2.597,-0.000", Scale="1,1,1") -> (Pose) {
    alphaInputType=Bool
    rotationMode=BMM_Additive
    rotationSpace=BCS_BoneSpace
    translationMode=BMM_Additive
    translationSpace=BCS_BoneSpace
    alphaBoolBlend {blendInTime=0.2, blendOption=Linear, blendOutTime=0.3}
    boneToModify {boneName=chain_01_back_02}
    rotation=<rot 0 -2.597 -0.0001>
    translation=<-0.352 -0.532>
   }
   n36 = "Transform (Modify) Bone - Bone: chain_01_front_04"!(ComponentPose=@n35, bAlphaBoolEnabled:bool=@n31, Translation="0.000,1.558,0.000", Rotation="0,0,0", Scale="1,1,1") -> (Pose) {
    alphaInputType=Bool
    rotationMode=BMM_Additive
    rotationSpace=BCS_BoneSpace
    translationMode=BMM_Additive
    translationSpace=BCS_BoneSpace
    alphaBoolBlend {blendInTime=0.2, blendOption=Linear, blendOutTime=0.3}
    boneToModify {boneName=chain_01_front_04}
    translation {y=1.558}
   }
   n37 = "Transform (Modify) Bone - Bone: chain_01_front_03"!(ComponentPose=@n36, bAlphaBoolEnabled:bool=@n1, Translation="2,0,2", Rotation="-6.396,-19.614,-3.012") -> (Pose) {
    alphaInputType=Bool
    rotationMode=BMM_Additive
    rotationSpace=BCS_BoneSpace
    translationMode=BMM_Additive
    translationSpace=BCS_BoneSpace
    alphaBoolBlend {blendInTime=0.1, blendOption=Linear, blendOutTime=0.2}
    boneToModify {boneName=chain_01_front_03}
    rotation=<rot -6.396 -19.614 -3.0121>
    translation {x=2, z=2}
   }
   n38 = "Transform (Modify) Bone - Bone: chain_01_front_04"!(ComponentPose=@n37, bAlphaBoolEnabled=@n1, Translation="0,0.5,1.5", Rotation="5.000,0.000,-0.000") -> (Pose) {
    alphaInputType=Bool
    rotationMode=BMM_Additive
    rotationSpace=BCS_BoneSpace
    translationMode=BMM_Additive
    translationSpace=BCS_BoneSpace
    alphaBoolBlend {blendInTime=0.1, blendOption=Linear, blendOutTime=0.2}
    boneToModify {boneName=chain_01_front_04}
    rotation=<rot 5 0 0>
    translation {y=0.5, z=1.5}
   }
   n39 = "Transform (Modify) Bone - Bone: chain_02_front_04"!(ComponentPose=@n38, bAlphaBoolEnabled:bool=@n1, Translation="0,0,1", Rotation="-5.000,0.000,-0.000") -> (Pose) {
    alphaInputType=Bool
    rotationMode=BMM_Additive
    rotationSpace=BCS_BoneSpace
    translationMode=BMM_Additive
    translationSpace=BCS_BoneSpace
    alphaBoolBlend {blendInTime=0.1, blendOption=Linear, blendOutTime=0.2}
    boneToModify {boneName=chain_02_front_04}
    rotation=<rot -5 0 0>
    translation {z=1}
   }
   n40 = "Transform (Modify) Bone - Bone: chain_01_front_03"!(ComponentPose=@n39, Alpha:float=1, Translation="0.026,0.003,0.228", Rotation="16.469,9.852,0.431", Scale:Vector="1,1,1") -> (Pose) {
    rotationMode=BMM_Additive
    rotationSpace=BCS_BoneSpace
    translationMode=BMM_Additive
    translationSpace=BCS_BoneSpace
    alphaBoolBlend {blendInTime=0.2, blendOption=Linear, blendOutTime=0.3}
    boneToModify {boneName=chain_01_front_03}
    rotation=<rot 16.469 9.852 0.431>
    translation=<0.026 0.003 0.228>
   }
   n41 = "Transform (Modify) Bone - Bone: chain_01_front_04"!(ComponentPose=@n40, Alpha:float=1, Translation="0.000,0.000,2.389", Rotation="6.472,4.935,1.081", Scale:Vector="1,1,1") -> (Pose) {
    rotationMode=BMM_Additive
    rotationSpace=BCS_BoneSpace
    translationMode=BMM_Additive
    translationSpace=BCS_BoneSpace
    alphaBoolBlend {blendInTime=0.2, blendOption=Linear, blendOutTime=0.3}
    boneToModify {boneName=chain_01_front_04}
    rotation=<rot 6.472 4.935 1.081>
    translation {z=2.389}
   }
   n42 = "Transform (Modify) Bone - Bone: chain_01_back_03"!(ComponentPose=@n41, Alpha:float=1, Translation="-0.158,-2.626,3.359", Rotation="-6.512,-0.000,0.000", Scale="1,1,1") -> (Pose) {
    rotationMode=BMM_Additive
    rotationSpace=BCS_BoneSpace
    translationMode=BMM_Additive
    translationSpace=BCS_BoneSpace
    alphaBoolBlend {blendInTime=0.2, blendOption=Linear, blendOutTime=0.3}
    boneToModify {boneName=chain_01_back_03}
    rotation=<rot -6.512 0 0>
    translation=<-0.158 -2.626 3.359>
   }
   n43 = "Transform (Modify) Bone - Bone: chain_01_back_02"!(ComponentPose=@n42, Alpha:float=1, Translation="0.217,-0.000,1.631", Rotation="9.101,-12.376,-1.625", Scale="1,1,1") -> (Pose) {
    rotationMode=BMM_Additive
    rotationSpace=BCS_BoneSpace
    translationMode=BMM_Additive
    translationSpace=BCS_BoneSpace
    alphaBoolBlend {blendInTime=0.2, blendOption=Linear, blendOutTime=0.3}
    boneToModify {boneName=chain_01_back_02}
    rotation=<rot 9.1011 -12.376 -1.625>
    translation {x=0.217, z=1.631}
   }
   n44 = "Transform (Modify) Bone - Bone: chain_01_back_01"(ComponentPose=@n43, Alpha=1, Translation="0,0,0", Rotation="5.525,0.000,-0.000", Scale="1,1,1") {
    rotationMode=BMM_Additive
    rotationSpace=BCS_BoneSpace
    translationMode=BMM_Additive
    translationSpace=BCS_BoneSpace
    alphaBoolBlend {blendInTime=0.2, blendOption=Linear, blendOutTime=0.3}
    boneToModify {boneName=chain_01_back_01}
    rotation=<rot 5.5251 0 0>
   }
   n45 = "Transform (Modify) Bone - Bone: chain_02_back_03"(ComponentPose=@n44, Alpha=1, Translation="0,0,0", Rotation="0,0,0", Scale="1,1,1") {
    rotationMode=BMM_Additive
    rotationSpace=BCS_BoneSpace
    translationMode=BMM_Additive
    translationSpace=BCS_BoneSpace
    alphaBoolBlend {blendInTime=0.2, blendOption=Linear, blendOutTime=0.3}
    boneToModify {boneName=chain_02_back_03}
   }
   n46 = "Transform (Modify) Bone - Bone: chain_03_front_03"!(ComponentPose=@n45, Alpha:float=1, Translation="0,0,0", Rotation="-17.150,18.212,-5.237", Scale="1,1,1") -> (Pose) {
    rotationMode=BMM_Additive
    rotationSpace=BCS_BoneSpace
    translationMode=BMM_Additive
    translationSpace=BCS_BoneSpace
    alphaBoolBlend {blendInTime=0.2, blendOption=Linear, blendOutTime=0.3}
    boneToModify {boneName=chain_03_front_03}
    rotation=<rot -17.15 18.212 -5.237>
   }
   n47 = "Transform (Modify) Bone - Bone: chain_03_front_04"!(ComponentPose=@n46, Alpha:float=1, Translation="-0.890,0.010,-2.753", Rotation="3.186,-3.478,-0.190", Scale="1,1,1") -> (Pose) {
    rotationMode=BMM_Additive
    rotationSpace=BCS_BoneSpace
    translationMode=BMM_Additive
    translationSpace=BCS_BoneSpace
    alphaBoolBlend {blendInTime=0.2, blendOption=Linear, blendOutTime=0.3}
    boneToModify {boneName=chain_03_front_04}
    rotation=<rot 3.186 -3.478 -0.19>
    translation=<-0.89 0.01 -2.753>
   }
   n48 = "Transform (Modify) Bone - Bone: chain_03_back_03"(ComponentPose=@n47, Alpha=1, Translation="0.102,-0.000,-0.358", Rotation="-11.949,-0.000,0.000", Scale="1,1,1") {
    rotationMode=BMM_Additive
    rotationSpace=BCS_BoneSpace
    translationMode=BMM_Additive
    translationSpace=BCS_BoneSpace
    alphaBoolBlend {blendInTime=0.2, blendOption=Linear, blendOutTime=0.3}
    boneToModify {boneName=chain_03_back_03}
    rotation=<rot -11.949 0 0>
    translation {x=0.102, z=-0.358}
   }
   Output Pose(Result=Component To Local(ComponentPose=@n48)) {layerGroup=None}
   n49 = Control Rig(Source=Copy Pose From Mesh() {bUseAttachedParent=true}) {
    alphaBoolBlend {blendOption=Linear}
    controlRigAssetReference {blueprintRigClass=&/M_MED_CupidEvil_CtrlRig1.M_MED_CupidEvil_CtrlRig1_C}
   }
   n50 = "Constraint - dyn_band_mstr (PPPP)"!(
    ComponentPose=Local To Component(LocalPose=@n49)
    Alpha=1
    ConstraintWeights_0=0.491
    ConstraintWeights_1:float=0.077
    ConstraintWeights_2:float=0.348
    ConstraintWeights_3:float=0.084
   ) -> (Pose) {
    alphaBoolBlend {blendOption=Linear}
    boneToModify {boneName=dyn_band_mstr}
    constraintSetup [
     {targetBone {boneName=upperarm_r}}
     {targetBone {boneName=upperarm_twist_01_r}}
     {targetBone {boneName=upperarm_twist_02_r}}
     {targetBone {boneName=lowerarm_r}}
    ]
    constraintWeights [1,1,1,1]
   }
   n51 = "Constraint - back_stone (PP)"!(ComponentPose=@n50, Alpha=1, ConstraintWeights_0=0.845, ConstraintWeights_1:float=0.155) -> (Pose) {
    alphaBoolBlend {blendOption=Linear}
    boneToModify {boneName=back_stone}
    constraintSetup [{targetBone {boneName=spine_05}}, {targetBone {boneName=clavicle_r}}]
    constraintWeights [1,1]
   }
   Save cached pose 'InputPose'(Pose=Component To Local(ComponentPose=@n51)) {cacheName=InputPose}
   n52 = "Transform (Modify) Bone - Bone: chain_03_front_base"!(ComponentPose=Local To Component(LocalPose=Use cached pose 'LeftPose'()), Alpha=1, Translation="0,0,0", Rotation="0,0,0") -> (Pose) {
    rotationMode=BMM_Additive
    rotationSpace=BCS_BoneSpace
    translationMode=BMM_Additive
    translationSpace=BCS_WorldSpace
    alphaBoolBlend {blendOption=Linear}
    boneToModify {boneName=chain_03_front_base}
   }
   n53 = "Transform (Modify) Bone - Bone: chain_03_front_01"!(ComponentPose=@n52, Alpha=1, Translation="0.642,-1.751,-1.612", Rotation="-34.234,-0.000,0.000") -> (Pose) {
    rotationMode=BMM_Additive
    rotationSpace=BCS_BoneSpace
    translationMode=BMM_Additive
    translationSpace=BCS_BoneSpace
    alphaBoolBlend {blendOption=Linear}
    boneToModify {boneName=chain_03_front_01}
    rotation=<rot -34.234 0 0.0001>
    translation=<0.642 -1.751 -1.612>
   }
   n54 = "Transform (Modify) Bone - Bone: chain_03_front_02"!(ComponentPose=@n53, Alpha=1, Translation="-3.869,-2.311,-6.206", Rotation="-43.669,16.858,-9.435") -> (Pose) {
    rotationMode=BMM_Additive
    rotationSpace=BCS_BoneSpace
    translationMode=BMM_Additive
    translationSpace=BCS_BoneSpace
    alphaBoolBlend {blendOption=Linear}
    boneToModify {boneName=chain_03_front_02}
    rotation=<rot -43.669 16.858 -9.435>
    translation=<-3.869 -2.311 -6.206>
   }
   n55 = "Transform (Modify) Bone - Bone: chain_03_front_03"!(ComponentPose=@n54, Alpha=1, Translation="-8.558,-1.762,-10.991", Rotation="-22.801,-33.595,14.546") -> (Pose) {
    rotationMode=BMM_Additive
    rotationSpace=BCS_BoneSpace
    translationMode=BMM_Additive
    translationSpace=BCS_BoneSpace
    alphaBoolBlend {blendOption=Linear}
    boneToModify {boneName=chain_03_front_03}
    rotation=<rot -22.801 -33.595 14.546>
    translation=<-8.558 -1.762 -10.991>
   }
   n56 = "Transform (Modify) Bone - Bone: chain_03_front_04"!(ComponentPose=@n55, Alpha=1, Translation="-0.570,-6.719,-17.493", Rotation="27.670,3.111,27.189") -> (Pose) {
    rotationMode=BMM_Additive
    rotationSpace=BCS_BoneSpace
    translationMode=BMM_Additive
    translationSpace=BCS_BoneSpace
    alphaBoolBlend {blendOption=Linear}
    boneToModify {boneName=chain_03_front_04}
    rotation=<rot 27.67 3.111 27.189>
    translation=<-0.57 -6.719 -17.493>
   }
   n57 = "Transform (Modify) Bone - Bone: chain_03_back_base"!(ComponentPose=@n56, Alpha=1, Translation="0,0,0", Rotation="0,0,0") -> (Pose) {
    rotationMode=BMM_Additive
    rotationSpace=BCS_BoneSpace
    translationMode=BMM_Additive
    translationSpace=BCS_BoneSpace
    alphaBoolBlend {blendOption=Linear}
    boneToModify {boneName=chain_03_back_base}
   }
   n58 = "Transform (Modify) Bone - Bone: chain_03_back_01"!(ComponentPose=@n57, Alpha=1, Translation="0,0,0", Rotation="0.330,0.000,0.000") -> (Pose) {
    rotationMode=BMM_Additive
    rotationSpace=BCS_BoneSpace
    translationMode=BMM_Additive
    translationSpace=BCS_BoneSpace
    alphaBoolBlend {blendOption=Linear}
    boneToModify {boneName=chain_03_back_01}
    rotation=<rot 0.33 0 0>
   }
   n59 = "Transform (Modify) Bone - Bone: chain_03_back_02"!(ComponentPose=@n58, Alpha=1, Translation="1.117,1.363,-1.518", Rotation="-45.627,12.849,-7.315") -> (Pose) {
    rotationMode=BMM_Additive
    rotationSpace=BCS_BoneSpace
    scaleSpace=BCS_BoneSpace
    translationMode=BMM_Additive
    translationSpace=BCS_BoneSpace
    alphaBoolBlend {blendOption=Linear}
    boneToModify {boneName=chain_03_back_02}
    rotation=<rot -45.627 12.849 -7.315>
    translation=<1.117 1.363 -1.518>
   }
   n60 = "Transform (Modify) Bone - Bone: chain_03_back_03"!(ComponentPose=@n59, Alpha=1, Translation="0.285,5.023,-6.679", Rotation="-40.265,12.850,-33.097") -> (Pose) {
    rotationMode=BMM_Additive
    rotationSpace=BCS_BoneSpace
    scaleSpace=BCS_BoneSpace
    translationMode=BMM_Additive
    translationSpace=BCS_BoneSpace
    alphaBoolBlend {blendOption=Linear}
    boneToModify {boneName=chain_03_back_03}
    rotation=<rot -40.265 12.85 -33.097>
    translation=<0.285 5.023 -6.679>
   }
   n61 = "Transform (Modify) Bone - Bone: chain_02_front_base"!(ComponentPose=@n60, Alpha=1, Translation="0,0,2", Rotation="57.397,-6.396,-10.314") -> (Pose) {
    rotationMode=BMM_Additive
    rotationSpace=BCS_BoneSpace
    translationMode=BMM_Additive
    translationSpace=BCS_BoneSpace
    alphaBoolBlend {blendOption=Linear}
    boneToModify {boneName=chain_02_front_base}
    rotation=<rot 57.397 -6.396 -10.314>
    translation {z=2}
   }
   n62 = "Transform (Modify) Bone - Bone: chain_02_front_01"!(ComponentPose=@n61, Alpha=1, Translation="0.092,-1.381,6.037", Rotation="16.375,9.306,0.825") -> (Pose) {
    rotationMode=BMM_Additive
    rotationSpace=BCS_BoneSpace
    translationMode=BMM_Additive
    translationSpace=BCS_BoneSpace
    alphaBoolBlend {blendOption=Linear}
    boneToModify {boneName=chain_02_front_01}
    rotation=<rot 16.375 9.306 0.825>
    translation=<0.092 -1.381 6.037>
   }
   n63 = "Transform (Modify) Bone - Bone: chain_02_front_02"!(ComponentPose=@n62, Alpha=1, Translation="2.226,1.593,7.492", Rotation="-26.779,-5.114,-0.716") -> (Pose) {
    rotationMode=BMM_Additive
    rotationSpace=BCS_BoneSpace
    translationMode=BMM_Additive
    translationSpace=BCS_BoneSpace
    alphaBoolBlend {blendOption=Linear}
    boneToModify {boneName=chain_02_front_02}
    rotation=<rot -26.779 -5.114 -0.716>
    translation=<2.226 1.593 7.492>
   }
   n64 = "Transform (Modify) Bone - Bone: chain_02_front_03"!(ComponentPose=@n63, Alpha=1, Translation="-2.539,0.001,1.911", Rotation="-6.354,9.246,-0.991") -> (Pose) {
    rotationMode=BMM_Additive
    rotationSpace=BCS_BoneSpace
    translationMode=BMM_Additive
    translationSpace=BCS_BoneSpace
    alphaBoolBlend {blendOption=Linear}
    boneToModify {boneName=chain_02_front_03}
    rotation=<rot -6.354 9.246 -0.991>
    translation=<-2.539 0.001 1.911>
   }
   n65 = "Transform (Modify) Bone - Bone: chain_02_front_04"!(ComponentPose=@n64, Alpha:float=1, Translation="0,0,0", Rotation="0,0,0") -> (Pose) {
    rotationMode=BMM_Additive
    rotationSpace=BCS_BoneSpace
    translationMode=BMM_Additive
    translationSpace=BCS_BoneSpace
    alphaBoolBlend {blendOption=Linear}
    boneToModify {boneName=chain_02_front_04}
   }
   n66 = "Transform (Modify) Bone - Bone: chain_02_back_base"!(ComponentPose=@n65, Alpha=1, Translation="0,0,0", Rotation="0,0,0") -> (Pose) {
    rotationMode=BMM_Additive
    rotationSpace=BCS_BoneSpace
    translationMode=BMM_Additive
    translationSpace=BCS_BoneSpace
    alphaBoolBlend {blendOption=Linear}
    boneToModify {boneName=chain_02_back_base}
   }
   n67 = "Transform (Modify) Bone - Bone: chain_02_back_01"!(ComponentPose=@n66, Alpha=1, Translation="0,0,0", Rotation="0,0,0") -> (Pose) {
    rotationMode=BMM_Additive
    rotationSpace=BCS_BoneSpace
    translationMode=BMM_Additive
    translationSpace=BCS_BoneSpace
    alphaBoolBlend {blendOption=Linear}
    boneToModify {boneName=chain_02_back_01}
   }
   n68 = "Transform (Modify) Bone - Bone: chain_02_back_02"!(ComponentPose=@n67, Alpha=1, Translation="0,0,0", Rotation="0,0,0") -> (Pose) {
    rotationMode=BMM_Additive
    rotationSpace=BCS_BoneSpace
    translationMode=BMM_Additive
    translationSpace=BCS_BoneSpace
    alphaBoolBlend {blendOption=Linear}
    boneToModify {boneName=chain_02_back_02}
   }
   n69 = "Transform (Modify) Bone - Bone: chain_02_back_03"!(ComponentPose=@n68, Alpha=1, Translation="0,0,0", Rotation="0,0,0") -> (Pose) {
    rotationMode=BMM_Additive
    rotationSpace=BCS_BoneSpace
    translationMode=BMM_Additive
    translationSpace=BCS_BoneSpace
    alphaBoolBlend {blendOption=Linear}
    boneToModify {boneName=chain_02_back_03}
   }
   n70 = "Transform (Modify) Bone - Bone: chain_01_front_base"!(ComponentPose=@n69, Alpha=1, Translation="0,0,1", Rotation="84.643,36.129,21.249") -> (Pose) {
    rotationMode=BMM_Additive
    rotationSpace=BCS_BoneSpace
    translationMode=BMM_Additive
    translationSpace=BCS_BoneSpace
    alphaBoolBlend {blendOption=Linear}
    boneToModify {boneName=chain_01_front_base}
    rotation=<rot 84.643 36.1289 21.2489>
    translation {z=1}
   }
   n71 = "Transform (Modify) Bone - Bone: chain_01_front_01"!(ComponentPose=@n70, Alpha=1, Translation="-3.427,-0.673,11.083", Rotation="6.640,14.266,1.744") -> (Pose) {
    rotationMode=BMM_Additive
    rotationSpace=BCS_BoneSpace
    translationMode=BMM_Additive
    translationSpace=BCS_BoneSpace
    alphaBoolBlend {blendOption=Linear}
    boneToModify {boneName=chain_01_front_01}
    rotation=<rot 6.64 14.266 1.744>
    translation=<-3.427 -0.673 11.083>
   }
   n72 = "Transform (Modify) Bone - Bone: chain_01_front_02"!(ComponentPose=@n71, Alpha=1, Translation="-1.786,0.853,11.849", Rotation="-25.946,31.335,-7.626") -> (Pose) {
    rotationMode=BMM_Additive
    rotationSpace=BCS_BoneSpace
    translationMode=BMM_Additive
    translationSpace=BCS_BoneSpace
    alphaBoolBlend {blendOption=Linear}
    boneToModify {boneName=chain_01_front_02}
    rotation=<rot -25.946 31.335 -7.626>
    translation=<-1.786 0.853 11.849>
   }
   n73 = "Transform (Modify) Bone - Bone: chain_01_front_03"!(ComponentPose=@n72, Alpha=1, Translation="-2.685,-5.382,7.575", Rotation="-41.659,-5.800,3.550") -> (Pose) {
    rotationMode=BMM_Additive
    rotationSpace=BCS_BoneSpace
    translationMode=BMM_Additive
    translationSpace=BCS_BoneSpace
    alphaBoolBlend {blendOption=Linear}
    boneToModify {boneName=chain_01_front_03}
    rotation=<rot -41.659 -5.8 3.55>
    translation=<-2.685 -5.382 7.575>
   }
   n74 = "Transform (Modify) Bone - Bone: chain_01_front_04"!(ComponentPose=@n73, Alpha=1, Translation="0,0,0", Rotation="0,0,0") -> (Pose) {
    rotationMode=BMM_Additive
    rotationSpace=BCS_BoneSpace
    translationMode=BMM_Additive
    translationSpace=BCS_BoneSpace
    alphaBoolBlend {blendOption=Linear}
    boneToModify {boneName=chain_01_front_04}
   }
   n75 = "Transform (Modify) Bone - Bone: chain_01_back_base"!(ComponentPose=@n74, Alpha=1, Translation="0,0,0", Rotation="0,0,0") -> (Pose) {
    rotationMode=BMM_Additive
    rotationSpace=BCS_BoneSpace
    translationMode=BMM_Additive
    translationSpace=BCS_BoneSpace
    alphaBoolBlend {blendOption=Linear}
    boneToModify {boneName=chain_01_back_base}
   }
   n76 = "Transform (Modify) Bone - Bone: chain_01_back_01"!(ComponentPose=@n75, Alpha=1, Translation="0,0,0", Rotation="0,0,0") -> (Pose) {
    rotationMode=BMM_Additive
    rotationSpace=BCS_BoneSpace
    translationMode=BMM_Additive
    translationSpace=BCS_BoneSpace
    alphaBoolBlend {blendOption=Linear}
    boneToModify {boneName=chain_01_back_01}
   }
   n77 = "Transform (Modify) Bone - Bone: chain_01_back_02"!(ComponentPose=@n76, Alpha=1, Translation="0,0,0", Rotation="0,0,0") -> (Pose) {
    rotationMode=BMM_Additive
    rotationSpace=BCS_BoneSpace
    translationMode=BMM_Additive
    translationSpace=BCS_BoneSpace
    alphaBoolBlend {blendOption=Linear}
    boneToModify {boneName=chain_01_back_02}
   }
   n78 = "Transform (Modify) Bone - Bone: chain_01_back_03"!(ComponentPose=@n77, Alpha=1, Translation="0,0,0", Rotation="0,0,0") -> (Pose) {
    rotationMode=BMM_Additive
    rotationSpace=BCS_BoneSpace
    translationMode=BMM_Additive
    translationSpace=BCS_BoneSpace
    alphaBoolBlend {blendOption=Linear}
    boneToModify {boneName=chain_01_back_03}
   }
   Save cached pose 'LeftDeformPose'(Pose=Component To Local(ComponentPose=@n78)) {cacheName=LeftDeformPose}
   n79 = "Transform (Modify) Bone - Bone: chain_03_front_base"(ComponentPose=Local To Component(LocalPose=Use cached pose 'RightPose'()), Alpha=1, Translation="0,0,3", Rotation="-29.908,4.610,-1.140") {
    rotationMode=BMM_Additive
    rotationSpace=BCS_BoneSpace
    translationMode=BMM_Additive
    translationSpace=BCS_WorldSpace
    alphaBoolBlend {blendOption=Linear}
    boneToModify {boneName=chain_03_front_base}
    rotation=<rot -29.908 4.61 -1.14>
    translation {z=3}
   }
   n80 = "Transform (Modify) Bone - Bone: chain_03_front_01"(ComponentPose=@n79, Alpha=1, Translation="2.504,-2.461,-5.356", Rotation="22.473,12.650,4.731") {
    rotationMode=BMM_Additive
    rotationSpace=BCS_BoneSpace
    translationMode=BMM_Additive
    translationSpace=BCS_BoneSpace
    alphaBoolBlend {blendOption=Linear}
    boneToModify {boneName=chain_03_front_01}
    rotation=<rot 22.473 12.65 4.731>
    translation=<2.504 -2.461 -5.356>
   }
   n81 = "Transform (Modify) Bone - Bone: chain_03_front_02"(ComponentPose=@n80, Alpha=1, Translation="-0.037,-0.990,-2.335", Rotation="20.346,4.283,1.474") {
    rotationMode=BMM_Additive
    rotationSpace=BCS_BoneSpace
    translationMode=BMM_Additive
    translationSpace=BCS_BoneSpace
    alphaBoolBlend {blendOption=Linear}
    boneToModify {boneName=chain_03_front_02}
    rotation=<rot 20.346 4.283 1.474>
    translation=<-0.037 -0.99 -2.335>
   }
   n82 = "Transform (Modify) Bone - Bone: chain_03_front_03"(ComponentPose=@n81, Alpha=1, Translation="0,0,0", Rotation="0.000,5.507,0.000") {
    rotationMode=BMM_Additive
    rotationSpace=BCS_BoneSpace
    translationMode=BMM_Additive
    translationSpace=BCS_BoneSpace
    alphaBoolBlend {blendOption=Linear}
    boneToModify {boneName=chain_03_front_03}
    rotation=<rot 0 5.507 0>
   }
   n83 = "Transform (Modify) Bone - Bone: chain_03_front_04"(ComponentPose=@n82, Alpha=1, Translation="0,0,0", Rotation="0.020,-5.447,5.599") {
    rotationMode=BMM_Additive
    rotationSpace=BCS_BoneSpace
    translationMode=BMM_Additive
    translationSpace=BCS_BoneSpace
    alphaBoolBlend {blendOption=Linear}
    boneToModify {boneName=chain_03_front_04}
    rotation=<rot 0.02 -5.447 5.599>
   }
   n84 = "Transform (Modify) Bone - Bone: chain_03_back_base"(ComponentPose=@n83, Alpha=1, Translation="0,0,0", Rotation="0,0,0") {
    rotationMode=BMM_Additive
    rotationSpace=BCS_BoneSpace
    translationMode=BMM_Additive
    translationSpace=BCS_BoneSpace
    alphaBoolBlend {blendOption=Linear}
    boneToModify {boneName=chain_03_back_base}
   }
   n85 = "Transform (Modify) Bone - Bone: chain_03_back_01"(ComponentPose=@n84, Alpha=1, Translation="0,0,0", Rotation="1.103,0.000,0.000") {
    rotationMode=BMM_Additive
    rotationSpace=BCS_BoneSpace
    translationMode=BMM_Additive
    translationSpace=BCS_BoneSpace
    alphaBoolBlend {blendOption=Linear}
    boneToModify {boneName=chain_03_back_01}
    rotation=<rot 1.103 0 0.0001>
   }
   n86 = "Transform (Modify) Bone - Bone: chain_03_back_02"(ComponentPose=@n85, Alpha=1, Translation="0.823,0.262,-0.044", Rotation="-5.395,5.140,1.758") {
    rotationMode=BMM_Additive
    rotationSpace=BCS_BoneSpace
    scaleSpace=BCS_BoneSpace
    translationMode=BMM_Additive
    translationSpace=BCS_BoneSpace
    alphaBoolBlend {blendOption=Linear}
    boneToModify {boneName=chain_03_back_02}
    rotation=<rot -5.395 5.14 1.758>
    translation=<0.823 0.262 -0.044>
   }
   n87 = "Transform (Modify) Bone - Bone: chain_03_back_03"(ComponentPose=@n86, Alpha=1, Translation="0.005,0.965,-0.560", Rotation="0.315,-0.477,0.000") {
    rotationMode=BMM_Additive
    rotationSpace=BCS_BoneSpace
    scaleSpace=BCS_BoneSpace
    translationMode=BMM_Additive
    translationSpace=BCS_BoneSpace
    alphaBoolBlend {blendOption=Linear}
    boneToModify {boneName=chain_03_back_03}
    rotation=<rot 0.315 -0.477 0>
    translation=<0.005 0.965 -0.56>
   }
   n88 = "Transform (Modify) Bone - Bone: chain_02_front_base"(ComponentPose=@n87, Alpha=1, Translation="0,0,0", Rotation="0,0,0") {
    rotationMode=BMM_Additive
    rotationSpace=BCS_BoneSpace
    translationMode=BMM_Additive
    translationSpace=BCS_BoneSpace
    alphaBoolBlend {blendOption=Linear}
    boneToModify {boneName=chain_02_front_base}
   }
   n89 = "Transform (Modify) Bone - Bone: chain_02_front_01"(ComponentPose=@n88, Alpha=1, Translation="0,0,0", Rotation="11.669,-5.319,-1.074") {
    rotationMode=BMM_Additive
    rotationSpace=BCS_BoneSpace
    translationMode=BMM_Additive
    translationSpace=BCS_BoneSpace
    alphaBoolBlend {blendOption=Linear}
    boneToModify {boneName=chain_02_front_01}
    rotation=<rot 11.669 -5.319 -1.074>
   }
   n90 = "Transform (Modify) Bone - Bone: chain_02_front_02"(ComponentPose=@n89, Alpha=1, Translation="-1.677,-0.998,2.893", Rotation="30.123,-0.045,-0.039") {
    rotationMode=BMM_Additive
    rotationSpace=BCS_BoneSpace
    translationMode=BMM_Additive
    translationSpace=BCS_BoneSpace
    alphaBoolBlend {blendOption=Linear}
    boneToModify {boneName=chain_02_front_02}
    rotation=<rot 30.123 -0.045 -0.039>
    translation=<-1.677 -0.998 2.893>
   }
   n91 = "Transform (Modify) Bone - Bone: chain_02_front_03"(ComponentPose=@n90, Alpha=1, Translation="-1.306,-0.134,6.779", Rotation="18.204,0.039,0.041") {
    rotationMode=BMM_Additive
    rotationSpace=BCS_BoneSpace
    translationMode=BMM_Additive
    translationSpace=BCS_BoneSpace
    alphaBoolBlend {blendOption=Linear}
    boneToModify {boneName=chain_02_front_03}
    rotation=<rot 18.204 0.039 0.041>
    translation=<-1.306 -0.134 6.779>
   }
   n92 = "Transform (Modify) Bone - Bone: chain_02_front_04"(ComponentPose=@n91, Alpha=1, Translation="1.513,-0.009,8.774", Rotation="-14.328,-11.851,1.912") {
    rotationMode=BMM_Additive
    rotationSpace=BCS_BoneSpace
    translationMode=BMM_Additive
    translationSpace=BCS_BoneSpace
    alphaBoolBlend {blendOption=Linear}
    boneToModify {boneName=chain_02_front_04}
    rotation=<rot -14.328 -11.851 1.912>
    translation=<1.513 -0.009 8.774>
   }
   n93 = "Transform (Modify) Bone - Bone: chain_02_back_base"(ComponentPose=@n92, Alpha=1, Translation="0,0,0", Rotation="0,0,0") {
    rotationMode=BMM_Additive
    rotationSpace=BCS_BoneSpace
    translationMode=BMM_Additive
    translationSpace=BCS_BoneSpace
    alphaBoolBlend {blendOption=Linear}
    boneToModify {boneName=chain_02_back_base}
   }
   n94 = "Transform (Modify) Bone - Bone: chain_02_back_01"(ComponentPose=@n93, Alpha=1, Translation="0,0,0", Rotation="0,0,0") {
    rotationMode=BMM_Additive
    rotationSpace=BCS_BoneSpace
    translationMode=BMM_Additive
    translationSpace=BCS_BoneSpace
    alphaBoolBlend {blendOption=Linear}
    boneToModify {boneName=chain_02_back_01}
   }
   n95 = "Transform (Modify) Bone - Bone: chain_02_back_02"(ComponentPose=@n94, Alpha=1, Translation="0,0,0", Rotation="22.981,3.811,1.487") {
    rotationMode=BMM_Additive
    rotationSpace=BCS_BoneSpace
    translationMode=BMM_Additive
    translationSpace=BCS_BoneSpace
    alphaBoolBlend {blendOption=Linear}
    boneToModify {boneName=chain_02_back_02}
    rotation=<rot 22.981 3.811 1.487>
   }
   n96 = "Transform (Modify) Bone - Bone: chain_02_back_03"(ComponentPose=@n95, Alpha=1, Translation="-1.267,0.306,4.434", Rotation="10.421,-2.426,-0.973") {
    rotationMode=BMM_Additive
    rotationSpace=BCS_BoneSpace
    translationMode=BMM_Additive
    translationSpace=BCS_BoneSpace
    alphaBoolBlend {blendOption=Linear}
    boneToModify {boneName=chain_02_back_03}
    rotation=<rot 10.4211 -2.426 -0.973>
    translation=<-1.267 0.306 4.434>
   }
   n97 = "Transform (Modify) Bone - Bone: chain_01_front_base"(ComponentPose=@n96, Alpha=1, Translation="0,0,0", Rotation="0,0,0") {
    rotationMode=BMM_Additive
    rotationSpace=BCS_BoneSpace
    translationMode=BMM_Additive
    translationSpace=BCS_BoneSpace
    alphaBoolBlend {blendOption=Linear}
    boneToModify {boneName=chain_01_front_base}
   }
   n98 = "Transform (Modify) Bone - Bone: chain_01_front_01"(ComponentPose=@n97, Alpha=1, Translation="2.625,0.769,0.245", Rotation="0,0,0") {
    rotationMode=BMM_Additive
    rotationSpace=BCS_BoneSpace
    translationMode=BMM_Additive
    translationSpace=BCS_BoneSpace
    alphaBoolBlend {blendOption=Linear}
    boneToModify {boneName=chain_01_front_01}
    translation=<2.625 0.769 0.245>
   }
   n99 = "Transform (Modify) Bone - Bone: chain_01_front_02"(ComponentPose=@n98, Alpha=1, Translation="2.445,0.697,2.862", Rotation="44.419,-0.000,-0.000") {
    rotationMode=BMM_Additive
    rotationSpace=BCS_BoneSpace
    translationMode=BMM_Additive
    translationSpace=BCS_BoneSpace
    alphaBoolBlend {blendOption=Linear}
    boneToModify {boneName=chain_01_front_02}
    rotation=<rot 44.419 0 0>
    translation=<2.445 0.697 2.862>
   }
   n100 = "Transform (Modify) Bone - Bone: chain_01_front_03"(ComponentPose=@n99, Alpha=1, Translation="0.143,1.760,6.955", Rotation="21.687,-5.694,2.808") {
    rotationMode=BMM_Additive
    rotationSpace=BCS_BoneSpace
    translationMode=BMM_Additive
    translationSpace=BCS_BoneSpace
    alphaBoolBlend {blendOption=Linear}
    boneToModify {boneName=chain_01_front_03}
    rotation=<rot 21.687 -5.694 2.808>
    translation=<0.143 1.76 6.955>
   }
   n101 = "Transform (Modify) Bone - Bone: chain_01_front_04"(ComponentPose=@n100, Alpha=1, Translation="1.196,1.637,11.288", Rotation="-12.305,-13.275,3.933") {
    rotationMode=BMM_Additive
    rotationSpace=BCS_BoneSpace
    translationMode=BMM_Additive
    translationSpace=BCS_BoneSpace
    alphaBoolBlend {blendOption=Linear}
    boneToModify {boneName=chain_01_front_04}
    rotation=<rot -12.305 -13.275 3.933>
    translation=<1.196 1.637 11.288>
   }
   n102 = "Transform (Modify) Bone - Bone: chain_01_back_base"(ComponentPose=@n101, Alpha=1, Translation="0,0,0", Rotation="0,0,0") {
    rotationMode=BMM_Additive
    rotationSpace=BCS_BoneSpace
    translationMode=BMM_Additive
    translationSpace=BCS_BoneSpace
    alphaBoolBlend {blendOption=Linear}
    boneToModify {boneName=chain_01_back_base}
   }
   n103 = "Transform (Modify) Bone - Bone: chain_01_back_01"(ComponentPose=@n102, Alpha=1, Translation="0.632,-0.429,0.154", Rotation="8.053,7.315,1.037") {
    rotationMode=BMM_Additive
    rotationSpace=BCS_BoneSpace
    translationMode=BMM_Additive
    translationSpace=BCS_BoneSpace
    alphaBoolBlend {blendOption=Linear}
    boneToModify {boneName=chain_01_back_01}
    rotation=<rot 8.053 7.315 1.037>
    translation=<0.632 -0.429 0.154>
   }
   n104 = "Transform (Modify) Bone - Bone: chain_01_back_02"(ComponentPose=@n103, Alpha=1, Translation="0.719,0.757,2.890", Rotation="12.619,12.672,4.650") {
    rotationMode=BMM_Additive
    rotationSpace=BCS_BoneSpace
    translationMode=BMM_Additive
    translationSpace=BCS_BoneSpace
    alphaBoolBlend {blendOption=Linear}
    boneToModify {boneName=chain_01_back_02}
    rotation=<rot 12.619 12.672 4.65>
    translation=<0.719 0.757 2.89>
   }
   n105 = "Transform (Modify) Bone - Bone: chain_01_back_03"(ComponentPose=@n104, Alpha=1, Translation="2.227,2.530,4.617", Rotation="28.709,-18.568,-9.617") {
    rotationMode=BMM_Additive
    rotationSpace=BCS_BoneSpace
    translationMode=BMM_Additive
    translationSpace=BCS_BoneSpace
    alphaBoolBlend {blendOption=Linear}
    boneToModify {boneName=chain_01_back_03}
    rotation=<rot 28.709 -18.568 -9.617>
    translation=<2.227 2.53 4.617>
   }
   Save cached pose 'RightDeformPose'(Pose=Component To Local(ComponentPose=@n105)) {cacheName=RightDeformPose}
   n106 = "Transform (Modify) Bone - Bone: chain_03_front_base"(ComponentPose=Local To Component(LocalPose=Use cached pose 'InputPose'()), Alpha=1, Translation="0,0,0", Rotation="-64.082,-0.000,0.000") {
    rotationMode=BMM_Additive
    rotationSpace=BCS_BoneSpace
    translationMode=BMM_Additive
    translationSpace=BCS_WorldSpace
    alphaBoolBlend {blendOption=Linear}
    boneToModify {boneName=chain_03_front_base}
    rotation=<rot -64.082 0 0.0001>
   }
   n107 = "Transform (Modify) Bone - Bone: chain_03_front_01"(ComponentPose=@n106, Alpha=1, Translation="-2.932,0.000,-7.677", Rotation="-54.511,-0.098,0.188") {
    rotationMode=BMM_Additive
    rotationSpace=BCS_BoneSpace
    translationMode=BMM_Additive
    translationSpace=BCS_BoneSpace
    alphaBoolBlend {blendOption=Linear}
    boneToModify {boneName=chain_03_front_01}
    rotation=<rot -54.511 -0.098 0.188>
    translation {x=-2.932, z=-7.677}
   }
   n108 = "Transform (Modify) Bone - Bone: chain_03_front_02"(ComponentPose=@n107, Alpha=1, Translation="-4.828,0.030,-16.669", Rotation="-39.825,-13.267,9.400") {
    rotationMode=BMM_Additive
    rotationSpace=BCS_BoneSpace
    translationMode=BMM_Additive
    translationSpace=BCS_BoneSpace
    alphaBoolBlend {blendOption=Linear}
    boneToModify {boneName=chain_03_front_02}
    rotation=<rot -39.825 -13.267 9.4001>
    translation=<-4.828 0.03 -16.669>
   }
   n109 = "Transform (Modify) Bone - Bone: chain_03_front_03"(ComponentPose=@n108, Alpha=1, Translation="-7.775,-1.513,-22.000", Rotation="-8.746,-16.034,10.979") {
    rotationMode=BMM_Additive
    rotationSpace=BCS_BoneSpace
    translationMode=BMM_Additive
    translationSpace=BCS_BoneSpace
    alphaBoolBlend {blendOption=Linear}
    boneToModify {boneName=chain_03_front_03}
    rotation=<rot -8.746 -16.034 10.979>
    translation=<-7.775 -1.513 -22>
   }
   n110 = "Transform (Modify) Bone - Bone: chain_03_front_04"(ComponentPose=@n109, Alpha=1, Translation="3.976,-1.675,-22.743", Rotation="35.207,-8.741,5.458") {
    rotationMode=BMM_Additive
    rotationSpace=BCS_BoneSpace
    translationMode=BMM_Additive
    translationSpace=BCS_BoneSpace
    alphaBoolBlend {blendOption=Linear}
    boneToModify {boneName=chain_03_front_04}
    rotation=<rot 35.207 -8.741 5.458>
    translation=<3.976 -1.675 -22.743>
   }
   n111 = "Transform (Modify) Bone - Bone: chain_03_back_base"(ComponentPose=@n110, Alpha=1, Translation="0,0,0", Rotation="0,0,0") {
    rotationMode=BMM_Additive
    rotationSpace=BCS_BoneSpace
    translationMode=BMM_Additive
    translationSpace=BCS_BoneSpace
    alphaBoolBlend {blendOption=Linear}
    boneToModify {boneName=chain_03_back_base}
   }
   n112 = "Transform (Modify) Bone - Bone: chain_03_back_01"(ComponentPose=@n111, Alpha=1, Translation="-2.078,-1.192,-0.482", Rotation="-16.971,-2.020,0.243") {
    rotationMode=BMM_Additive
    rotationSpace=BCS_BoneSpace
    translationMode=BMM_Additive
    translationSpace=BCS_BoneSpace
    alphaBoolBlend {blendOption=Linear}
    boneToModify {boneName=chain_03_back_01}
    rotation=<rot -16.971 -2.02 0.2431>
    translation=<-2.078 -1.192 -0.482>
   }
   n113 = "Transform (Modify) Bone - Bone: chain_03_back_02"(ComponentPose=@n112, Alpha=1, Translation="-3.922,0.135,-5.948", Rotation="-29.870,-16.079,7.033") {
    rotationMode=BMM_Additive
    rotationSpace=BCS_BoneSpace
    scaleSpace=BCS_BoneSpace
    translationMode=BMM_Additive
    translationSpace=BCS_BoneSpace
    alphaBoolBlend {blendOption=Linear}
    boneToModify {boneName=chain_03_back_02}
    rotation=<rot -29.87 -16.079 7.033>
    translation=<-3.922 0.135 -5.948>
   }
   n114 = "Transform (Modify) Bone - Bone: chain_03_back_03"(ComponentPose=@n113, Alpha=1, Translation="-4.053,-0.513,-11.823", Rotation="-34.413,-4.253,-5.754") {
    rotationMode=BMM_Additive
    rotationSpace=BCS_BoneSpace
    scaleSpace=BCS_BoneSpace
    translationMode=BMM_Additive
    translationSpace=BCS_BoneSpace
    alphaBoolBlend {blendOption=Linear}
    boneToModify {boneName=chain_03_back_03}
    rotation=<rot -34.413 -4.253 -5.754>
    translation=<-4.053 -0.513 -11.823>
   }
   n115 = "Transform (Modify) Bone - Bone: chain_02_front_base"(ComponentPose=@n114, Alpha=1, Translation="0,0,2", Rotation="63.863,13.033,10.606") {
    rotationMode=BMM_Additive
    rotationSpace=BCS_BoneSpace
    translationMode=BMM_Additive
    translationSpace=BCS_BoneSpace
    alphaBoolBlend {blendOption=Linear}
    boneToModify {boneName=chain_02_front_base}
    rotation=<rot 63.863 13.0329 10.6059>
    translation {z=2}
   }
   n116 = "Transform (Modify) Bone - Bone: chain_02_front_01"(ComponentPose=@n115, Alpha=1, Translation="0.232,0.278,6.184", Rotation="26.132,9.101,0.823") {
    rotationMode=BMM_Additive
    rotationSpace=BCS_BoneSpace
    translationMode=BMM_Additive
    translationSpace=BCS_BoneSpace
    alphaBoolBlend {blendOption=Linear}
    boneToModify {boneName=chain_02_front_01}
    rotation=<rot 26.132 9.101 0.823>
    translation=<0.232 0.278 6.184>
   }
   n117 = "Transform (Modify) Bone - Bone: chain_02_front_02"(ComponentPose=@n116, Alpha=1, Translation="0.735,1.712,8.217", Rotation="-10.595,-5.310,-0.702") {
    rotationMode=BMM_Additive
    rotationSpace=BCS_BoneSpace
    translationMode=BMM_Additive
    translationSpace=BCS_BoneSpace
    alphaBoolBlend {blendOption=Linear}
    boneToModify {boneName=chain_02_front_02}
    rotation=<rot -10.595 -5.31 -0.702>
    translation=<0.735 1.712 8.217>
   }
   n118 = "Transform (Modify) Bone - Bone: chain_02_front_03"(ComponentPose=@n117, Alpha=1, Translation="0.008,2.511,8.506", Rotation="-17.591,-13.843,-2.325") {
    rotationMode=BMM_Additive
    rotationSpace=BCS_BoneSpace
    translationMode=BMM_Additive
    translationSpace=BCS_BoneSpace
    alphaBoolBlend {blendOption=Linear}
    boneToModify {boneName=chain_02_front_03}
    rotation=<rot -17.591 -13.843 -2.325>
    translation=<0.008 2.511 8.506>
   }
   n119 = "Transform (Modify) Bone - Bone: chain_02_front_04"(ComponentPose=@n118, Alpha=1, Translation="0.256,2.512,5.748", Rotation="-12.476,-5.951,0.120") {
    rotationMode=BMM_Additive
    rotationSpace=BCS_BoneSpace
    translationMode=BMM_Additive
    translationSpace=BCS_BoneSpace
    alphaBoolBlend {blendOption=Linear}
    boneToModify {boneName=chain_02_front_04}
    rotation=<rot -12.476 -5.951 0.12>
    translation=<0.256 2.512 5.748>
   }
   n120 = "Transform (Modify) Bone - Bone: chain_02_back_base"(ComponentPose=@n119, Alpha=1, Translation="0,0,0", Rotation="0,0,0") {
    rotationMode=BMM_Additive
    rotationSpace=BCS_BoneSpace
    translationMode=BMM_Additive
    translationSpace=BCS_BoneSpace
    alphaBoolBlend {blendOption=Linear}
    boneToModify {boneName=chain_02_back_base}
   }
   n121 = "Transform (Modify) Bone - Bone: chain_02_back_01"(ComponentPose=@n120, Alpha=1, Translation="0,0,0", Rotation="4.815,0.000,-0.000") {
    rotationMode=BMM_Additive
    rotationSpace=BCS_BoneSpace
    translationMode=BMM_Additive
    translationSpace=BCS_BoneSpace
    alphaBoolBlend {blendOption=Linear}
    boneToModify {boneName=chain_02_back_01}
    rotation=<rot 4.815 0 0>
   }
   n122 = "Transform (Modify) Bone - Bone: chain_02_back_02"(ComponentPose=@n121, Alpha=1, Translation="0.422,-0.008,1.328", Rotation="4.830,-9.099,-4.934") {
    rotationMode=BMM_Additive
    rotationSpace=BCS_BoneSpace
    translationMode=BMM_Additive
    translationSpace=BCS_BoneSpace
    alphaBoolBlend {blendOption=Linear}
    boneToModify {boneName=chain_02_back_02}
    rotation=<rot 4.83 -9.099 -4.934>
    translation=<0.422 -0.008 1.328>
   }
   n123 = "Transform (Modify) Bone - Bone: chain_02_back_03"(ComponentPose=@n122, Alpha=1, Translation="1.407,-1.818,1.858", Rotation="10.691,0.000,-0.000") {
    rotationMode=BMM_Additive
    rotationSpace=BCS_BoneSpace
    translationMode=BMM_Additive
    translationSpace=BCS_BoneSpace
    alphaBoolBlend {blendOption=Linear}
    boneToModify {boneName=chain_02_back_03}
    rotation=<rot 10.691 0 0>
    translation=<1.407 -1.818 1.858>
   }
   n124 = "Transform (Modify) Bone - Bone: chain_01_front_base"(ComponentPose=@n123, Alpha=1, Translation="0,0,1", Rotation="82.379,29.098,9.287") {
    rotationMode=BMM_Additive
    rotationSpace=BCS_BoneSpace
    translationMode=BMM_Additive
    translationSpace=BCS_BoneSpace
    alphaBoolBlend {blendOption=Linear}
    boneToModify {boneName=chain_01_front_base}
    rotation=<rot 82.379 29.0979 9.2868>
    translation {z=1}
   }
   n125 = "Transform (Modify) Bone - Bone: chain_01_front_01"(ComponentPose=@n124, Alpha=1, Translation="-6.034,1.451,9.386", Rotation="29.437,7.148,-1.269") {
    rotationMode=BMM_Additive
    rotationSpace=BCS_BoneSpace
    translationMode=BMM_Additive
    translationSpace=BCS_BoneSpace
    alphaBoolBlend {blendOption=Linear}
    boneToModify {boneName=chain_01_front_01}
    rotation=<rot 29.437 7.148 -1.269>
    translation=<-6.034 1.451 9.386>
   }
   n126 = "Transform (Modify) Bone - Bone: chain_01_front_02"(ComponentPose=@n125, Alpha=1, Translation="-6.869,3.298,13.130", Rotation="-12.326,6.265,-2.088") {
    rotationMode=BMM_Additive
    rotationSpace=BCS_BoneSpace
    translationMode=BMM_Additive
    translationSpace=BCS_BoneSpace
    alphaBoolBlend {blendOption=Linear}
    boneToModify {boneName=chain_01_front_02}
    rotation=<rot -12.326 6.265 -2.088>
    translation=<-6.869 3.298 13.13>
   }
   n127 = "Transform (Modify) Bone - Bone: chain_01_front_03"(ComponentPose=@n126, Alpha=1, Translation="-5.243,4.655,11.649", Rotation="-9.302,-1.074,-8.434") {
    rotationMode=BMM_Additive
    rotationSpace=BCS_BoneSpace
    translationMode=BMM_Additive
    translationSpace=BCS_BoneSpace
    alphaBoolBlend {blendOption=Linear}
    boneToModify {boneName=chain_01_front_03}
    rotation=<rot -9.302 -1.074 -8.434>
    translation=<-5.243 4.655 11.649>
   }
   n128 = "Transform (Modify) Bone - Bone: chain_01_front_04"(ComponentPose=@n127, Alpha=1, Translation="-3.311,3.285,8.988", Rotation="-15.045,-2.247,0.565") {
    rotationMode=BMM_Additive
    rotationSpace=BCS_BoneSpace
    translationMode=BMM_Additive
    translationSpace=BCS_BoneSpace
    alphaBoolBlend {blendOption=Linear}
    boneToModify {boneName=chain_01_front_04}
    rotation=<rot -15.045 -2.247 0.565>
    translation=<-3.311 3.285 8.988>
   }
   n129 = "Transform (Modify) Bone - Bone: chain_01_back_base"(ComponentPose=@n128, Alpha=1, Translation="0,0,0", Rotation="0,0,0") {
    rotationMode=BMM_Additive
    rotationSpace=BCS_BoneSpace
    translationMode=BMM_Additive
    translationSpace=BCS_BoneSpace
    alphaBoolBlend {blendOption=Linear}
    boneToModify {boneName=chain_01_back_base}
   }
   n130 = "Transform (Modify) Bone - Bone: chain_01_back_01"(ComponentPose=@n129, Alpha=1, Translation="0.724,0.026,0.065", Rotation="5.250,0.000,-0.000") {
    rotationMode=BMM_Additive
    rotationSpace=BCS_BoneSpace
    translationMode=BMM_Additive
    translationSpace=BCS_BoneSpace
    alphaBoolBlend {blendOption=Linear}
    boneToModify {boneName=chain_01_back_01}
    rotation=<rot 5.25 0 0>
    translation=<0.724 0.026 0.065>
   }
   n131 = "Transform (Modify) Bone - Bone: chain_01_back_02"(ComponentPose=@n130, Alpha=1, Translation="2.307,-0.678,1.954", Rotation="13.600,9.783,1.223") {
    rotationMode=BMM_Additive
    rotationSpace=BCS_BoneSpace
    translationMode=BMM_Additive
    translationSpace=BCS_BoneSpace
    alphaBoolBlend {blendOption=Linear}
    boneToModify {boneName=chain_01_back_02}
    rotation=<rot 13.6 9.783 1.223>
    translation=<2.3071 -0.6779 1.954>
   }
   n132 = "Transform (Modify) Bone - Bone: chain_01_back_03"(ComponentPose=@n131, Alpha=1, Translation="3.735,-2.092,4.048", Rotation="11.383,9.789,-0.271") {
    rotationMode=BMM_Additive
    rotationSpace=BCS_BoneSpace
    translationMode=BMM_Additive
    translationSpace=BCS_BoneSpace
    alphaBoolBlend {blendOption=Linear}
    boneToModify {boneName=chain_01_back_03}
    rotation=<rot 11.383 9.789 -0.271>
    translation=<3.735 -2.092 4.048>
   }
   Save cached pose 'RightPose'(Pose=Component To Local(ComponentPose=@n132)) {cacheName=RightPose}
   n133 = "Transform (Modify) Bone - Bone: chain_03_front_base"(ComponentPose=Local To Component(LocalPose=Use cached pose 'InputPose'()), Alpha=1, Translation="0,0,2", Rotation="-46.671,-7.584,37.295") {
    rotationMode=BMM_Additive
    rotationSpace=BCS_BoneSpace
    translationMode=BMM_Additive
    translationSpace=BCS_WorldSpace
    alphaBoolBlend {blendOption=Linear}
    boneToModify {boneName=chain_03_front_base}
    rotation=<rot -46.671 -7.584 37.295>
    translation {z=2}
   }
   n134 = "Transform (Modify) Bone - Bone: chain_03_front_01"(ComponentPose=@n133, Alpha=1, Translation="-5.466,1.453,-11.249", Rotation="-36.882,0.011,-0.005") {
    rotationMode=BMM_Additive
    rotationSpace=BCS_BoneSpace
    translationMode=BMM_Additive
    translationSpace=BCS_BoneSpace
    alphaBoolBlend {blendOption=Linear}
    boneToModify {boneName=chain_03_front_01}
    rotation=<rot -36.882 0.011 -0.005>
    translation=<-5.466 1.453 -11.249>
   }
   n135 = "Transform (Modify) Bone - Bone: chain_03_front_02"(ComponentPose=@n134, Alpha=1, Translation="-6.606,1.347,-16.705", Rotation="19.975,5.998,0.973") {
    rotationMode=BMM_Additive
    rotationSpace=BCS_BoneSpace
    translationMode=BMM_Additive
    translationSpace=BCS_BoneSpace
    alphaBoolBlend {blendOption=Linear}
    boneToModify {boneName=chain_03_front_02}
    rotation=<rot 19.975 5.998 0.973>
    translation=<-6.606 1.347 -16.705>
   }
   n136 = "Transform (Modify) Bone - Bone: chain_03_front_03"(ComponentPose=@n135, Alpha=1, Translation="-5.949,2.422,-13.217", Rotation="24.849,6.442,-5.740") {
    rotationMode=BMM_Additive
    rotationSpace=BCS_BoneSpace
    translationMode=BMM_Additive
    translationSpace=BCS_BoneSpace
    alphaBoolBlend {blendOption=Linear}
    boneToModify {boneName=chain_03_front_03}
    rotation=<rot 24.849 6.4421 -5.7399>
    translation=<-5.949 2.422 -13.217>
   }
   n137 = "Transform (Modify) Bone - Bone: chain_03_front_04"(ComponentPose=@n136, Alpha=1, Translation="-1.212,0.410,-7.532", Rotation="14.290,-7.820,-31.309") {
    rotationMode=BMM_Additive
    rotationSpace=BCS_BoneSpace
    translationMode=BMM_Additive
    translationSpace=BCS_BoneSpace
    alphaBoolBlend {blendOption=Linear}
    boneToModify {boneName=chain_03_front_04}
    rotation=<rot 14.29 -7.82 -31.309>
    translation=<-1.212 0.41 -7.532>
   }
   n138 = "Transform (Modify) Bone - Bone: chain_03_back_base"(ComponentPose=@n137, Alpha=1, Translation="0,0,2", Rotation="0,0,0") {
    rotationMode=BMM_Additive
    rotationSpace=BCS_BoneSpace
    translationMode=BMM_Additive
    translationSpace=BCS_WorldSpace
    alphaBoolBlend {blendOption=Linear}
    boneToModify {boneName=chain_03_back_base}
    translation {z=2}
   }
   n139 = "Transform (Modify) Bone - Bone: chain_03_back_01"(ComponentPose=@n138, Alpha=1, Translation="-0.138,-0.875,-0.935", Rotation="-8.229,-0.011,0.000") {
    rotationMode=BMM_Additive
    rotationSpace=BCS_BoneSpace
    translationMode=BMM_Additive
    translationSpace=BCS_BoneSpace
    alphaBoolBlend {blendOption=Linear}
    boneToModify {boneName=chain_03_back_01}
    rotation=<rot -8.229 -0.011 0.0001>
    translation=<-0.138 -0.875 -0.935>
   }
   n140 = "Transform (Modify) Bone - Bone: chain_03_back_02"(ComponentPose=@n139, Alpha=1, Translation="-0.070,-0.422,-4.013", Rotation="-4.302,4.834,19.755") {
    rotationMode=BMM_Additive
    rotationSpace=BCS_BoneSpace
    scaleSpace=BCS_BoneSpace
    translationMode=BMM_Additive
    translationSpace=BCS_BoneSpace
    alphaBoolBlend {blendOption=Linear}
    boneToModify {boneName=chain_03_back_02}
    rotation=<rot -4.302 4.834 19.755>
    translation=<-0.07 -0.422 -4.013>
   }
   n141 = "Transform (Modify) Bone - Bone: chain_03_back_03"(ComponentPose=@n140, Alpha=1, Translation="0.267,-0.132,-5.395", Rotation="-7.429,5.417,37.146") {
    rotationMode=BMM_Additive
    rotationSpace=BCS_BoneSpace
    scaleSpace=BCS_BoneSpace
    translationMode=BMM_Additive
    translationSpace=BCS_BoneSpace
    alphaBoolBlend {blendOption=Linear}
    boneToModify {boneName=chain_03_back_03}
    rotation=<rot -7.429 5.417 37.146>
    translation=<0.267 -0.132 -5.395>
   }
   n142 = "Transform (Modify) Bone - Bone: chain_02_front_base"(ComponentPose=@n141, Alpha=1, Translation="0,0,0", Rotation="9.843,2.571,-4.067") {
    rotationMode=BMM_Additive
    rotationSpace=BCS_BoneSpace
    translationMode=BMM_Additive
    translationSpace=BCS_BoneSpace
    alphaBoolBlend {blendOption=Linear}
    boneToModify {boneName=chain_02_front_base}
    rotation=<rot 9.843 2.571 -4.067>
   }
   n143 = "Transform (Modify) Bone - Bone: chain_02_front_01"(ComponentPose=@n142, Alpha=1, Translation="-0.225,1.899,2.464", Rotation="25.998,-4.936,-1.454") {
    rotationMode=BMM_Additive
    rotationSpace=BCS_BoneSpace
    translationMode=BMM_Additive
    translationSpace=BCS_BoneSpace
    alphaBoolBlend {blendOption=Linear}
    boneToModify {boneName=chain_02_front_01}
    rotation=<rot 25.998 -4.936 -1.454>
    translation=<-0.225 1.899 2.464>
   }
   n144 = "Transform (Modify) Bone - Bone: chain_02_front_02"(ComponentPose=@n143, Alpha=1, Translation="0.262,0.351,6.587", Rotation="43.993,-0.139,0.120") {
    rotationMode=BMM_Additive
    rotationSpace=BCS_BoneSpace
    translationMode=BMM_Additive
    translationSpace=BCS_BoneSpace
    alphaBoolBlend {blendOption=Linear}
    boneToModify {boneName=chain_02_front_02}
    rotation=<rot 43.993 -0.139 0.12>
    translation=<0.262 0.351 6.587>
   }
   n145 = "Transform (Modify) Bone - Bone: chain_02_front_03"(ComponentPose=@n144, Alpha=1, Translation="0.681,-0.238,14.164", Rotation="6.495,-10.124,0.728") {
    rotationMode=BMM_Additive
    rotationSpace=BCS_BoneSpace
    translationMode=BMM_Additive
    translationSpace=BCS_BoneSpace
    alphaBoolBlend {blendOption=Linear}
    boneToModify {boneName=chain_02_front_03}
    rotation=<rot 6.495 -10.124 0.728>
    translation=<0.681 -0.238 14.164>
   }
   n146 = "Transform (Modify) Bone - Bone: chain_02_front_04"(ComponentPose=@n145, Alpha=1, Translation="3.374,2.021,14.051", Rotation="-30.245,-19.629,5.458") {
    rotationMode=BMM_Additive
    rotationSpace=BCS_BoneSpace
    translationMode=BMM_Additive
    translationSpace=BCS_BoneSpace
    alphaBoolBlend {blendOption=Linear}
    boneToModify {boneName=chain_02_front_04}
    rotation=<rot -30.245 -19.629 5.458>
    translation=<3.374 2.021 14.051>
   }
   n147 = "Transform (Modify) Bone - Bone: chain_02_back_base"(ComponentPose=@n146, Alpha=1, Translation="0,0,0", Rotation="0,0,0") {
    rotationMode=BMM_Additive
    rotationSpace=BCS_BoneSpace
    translationMode=BMM_Additive
    translationSpace=BCS_BoneSpace
    alphaBoolBlend {blendOption=Linear}
    boneToModify {boneName=chain_02_back_base}
   }
   n148 = "Transform (Modify) Bone - Bone: chain_02_back_01"(ComponentPose=@n147, Alpha=1, Translation="0,0,0", Rotation="4.832,0.000,-0.000") {
    rotationMode=BMM_Additive
    rotationSpace=BCS_BoneSpace
    translationMode=BMM_Additive
    translationSpace=BCS_BoneSpace
    alphaBoolBlend {blendOption=Linear}
    boneToModify {boneName=chain_02_back_01}
    rotation=<rot 4.832 0 0>
   }
   n149 = "Transform (Modify) Bone - Bone: chain_02_back_02"(ComponentPose=@n148, Alpha=1, Translation="0.384,0.011,1.133", Rotation="12.176,3.744,1.123") {
    rotationMode=BMM_Additive
    rotationSpace=BCS_BoneSpace
    translationMode=BMM_Additive
    translationSpace=BCS_BoneSpace
    alphaBoolBlend {blendOption=Linear}
    boneToModify {boneName=chain_02_back_02}
    rotation=<rot 12.176 3.744 1.123>
    translation=<0.384 0.011 1.1328>
   }
   n150 = "Transform (Modify) Bone - Bone: chain_02_back_03"(ComponentPose=@n149, Alpha=1, Translation="-0.016,0.275,3.267", Rotation="31.477,-13.441,-7.064") {
    rotationMode=BMM_Additive
    rotationSpace=BCS_BoneSpace
    translationMode=BMM_Additive
    translationSpace=BCS_BoneSpace
    alphaBoolBlend {blendOption=Linear}
    boneToModify {boneName=chain_02_back_03}
    rotation=<rot 31.477 -13.441 -7.064>
    translation=<-0.016 0.275 3.267>
   }
   n151 = "Transform (Modify) Bone - Bone: chain_01_front_base"(ComponentPose=@n150, Alpha=1, Translation="0,0,0", Rotation="14.723,4.622,-1.759") {
    rotationMode=BMM_Additive
    rotationSpace=BCS_BoneSpace
    translationMode=BMM_Additive
    translationSpace=BCS_BoneSpace
    alphaBoolBlend {blendOption=Linear}
    boneToModify {boneName=chain_01_front_base}
    rotation=<rot 14.723 4.622 -1.759>
   }
   n152 = "Transform (Modify) Bone - Bone: chain_01_front_01"(ComponentPose=@n151, Alpha=1, Translation="1.841,0.862,3.590", Rotation="47.928,10.342,6.081") {
    rotationMode=BMM_Additive
    rotationSpace=BCS_BoneSpace
    translationMode=BMM_Additive
    translationSpace=BCS_BoneSpace
    alphaBoolBlend {blendOption=Linear}
    boneToModify {boneName=chain_01_front_01}
    rotation=<rot 47.928 10.342 6.081>
    translation=<1.841 0.862 3.59>
   }
   n153 = "Transform (Modify) Bone - Bone: chain_01_front_02"(ComponentPose=@n152, Alpha=1, Translation="1.444,1.413,11.894", Rotation="55.166,10.583,7.095") {
    rotationMode=BMM_Additive
    rotationSpace=BCS_BoneSpace
    translationMode=BMM_Additive
    translationSpace=BCS_BoneSpace
    alphaBoolBlend {blendOption=Linear}
    boneToModify {boneName=chain_01_front_02}
    rotation=<rot 55.166 10.583 7.095>
    translation=<1.444 1.413 11.894>
   }
   n154 = "Transform (Modify) Bone - Bone: chain_01_front_03"(ComponentPose=@n153, Alpha=1, Translation="-3.052,4.252,15.474", Rotation="28.240,5.096,6.311") {
    rotationMode=BMM_Additive
    rotationSpace=BCS_BoneSpace
    translationMode=BMM_Additive
    translationSpace=BCS_BoneSpace
    alphaBoolBlend {blendOption=Linear}
    boneToModify {boneName=chain_01_front_03}
    rotation=<rot 28.24 5.096 6.311>
    translation=<-3.052 4.252 15.474>
   }
   n155 = "Transform (Modify) Bone - Bone: chain_01_front_04"(ComponentPose=@n154, Alpha=1, Translation="-0.515,2.673,19.008", Rotation="-28.865,-8.864,6.063") {
    rotationMode=BMM_Additive
    rotationSpace=BCS_BoneSpace
    translationMode=BMM_Additive
    translationSpace=BCS_BoneSpace
    alphaBoolBlend {blendOption=Linear}
    boneToModify {boneName=chain_01_front_04}
    rotation=<rot -28.865 -8.864 6.063>
    translation=<-0.515 2.673 19.008>
   }
   n156 = "Transform (Modify) Bone - Bone: chain_01_back_base"(ComponentPose=@n155, Alpha=1, Translation="0,0,0", Rotation="0,0,0") {
    rotationMode=BMM_Additive
    rotationSpace=BCS_BoneSpace
    translationMode=BMM_Additive
    translationSpace=BCS_BoneSpace
    alphaBoolBlend {blendOption=Linear}
    boneToModify {boneName=chain_01_back_base}
   }
   n157 = "Transform (Modify) Bone - Bone: chain_01_back_01"(ComponentPose=@n156, Alpha=1, Translation="1.914,-0.365,0.361", Rotation="9.406,7.271,0") {
    rotationMode=BMM_Additive
    rotationSpace=BCS_BoneSpace
    translationMode=BMM_Additive
    translationSpace=BCS_BoneSpace
    alphaBoolBlend {blendOption=Linear}
    boneToModify {boneName=chain_01_back_01}
    rotation {pitch=9.406, yaw=7.271}
    translation=<1.914 -0.365 0.361>
   }
   n158 = "Transform (Modify) Bone - Bone: chain_01_back_02"(ComponentPose=@n157, Alpha=1, Translation="3.817,1.446,4.026", Rotation="33.273,14.296,0") {
    rotationMode=BMM_Additive
    rotationSpace=BCS_BoneSpace
    translationMode=BMM_Additive
    translationSpace=BCS_BoneSpace
    alphaBoolBlend {blendOption=Linear}
    boneToModify {boneName=chain_01_back_02}
    rotation {pitch=33.273, yaw=14.296}
    translation=<3.817 1.446 4.026>
   }
   n159 = "Transform (Modify) Bone - Bone: chain_01_back_03"(ComponentPose=@n158, Alpha=1, Translation="4.706,-1.776,8.478", Rotation="48.538,3.519,5.074") {
    rotationMode=BMM_Additive
    rotationSpace=BCS_BoneSpace
    translationMode=BMM_Additive
    translationSpace=BCS_BoneSpace
    alphaBoolBlend {blendOption=Linear}
    boneToModify {boneName=chain_01_back_03}
    rotation=<rot 48.5381 3.519 5.074>
    translation=<4.706 -1.776 8.478>
   }
   n160 = "Transform (Modify) Bone - Bone: chain_03_front_base"(ComponentPose=@n159, Alpha=1, Translation="0,0,0", Rotation="-2.238,-3.647,0.023") {
    rotationMode=BMM_Additive
    rotationSpace=BCS_BoneSpace
    translationMode=BMM_Additive
    translationSpace=BCS_WorldSpace
    alphaBoolBlend {blendOption=Linear}
    boneToModify {boneName=chain_03_front_base}
    rotation=<rot -2.238 -3.647 0.023>
   }
   n161 = "Transform (Modify) Bone - Bone: chain_03_front_01"(ComponentPose=@n160, Alpha=1, Translation="1.581,-2.344,-3.462", Rotation="5.082,-5.699,22.170") {
    rotationMode=BMM_Additive
    rotationSpace=BCS_BoneSpace
    translationMode=BMM_Additive
    translationSpace=BCS_BoneSpace
    alphaBoolBlend {blendOption=Linear}
    boneToModify {boneName=chain_03_front_01}
    rotation=<rot 5.082 -5.699 22.17>
    translation=<1.581 -2.344 -3.462>
   }
   n162 = "Transform (Modify) Bone - Bone: chain_03_front_02"(ComponentPose=@n161, Alpha=1, Translation="-2.392,-3.660,-5.132", Rotation="-11.468,4.678,1.916") {
    rotationMode=BMM_Additive
    rotationSpace=BCS_BoneSpace
    translationMode=BMM_Additive
    translationSpace=BCS_BoneSpace
    alphaBoolBlend {blendOption=Linear}
    boneToModify {boneName=chain_03_front_02}
    rotation=<rot -11.468 4.678 1.916>
    translation=<-2.392 -3.66 -5.132>
   }
   n163 = "Transform (Modify) Bone - Bone: chain_03_front_03"(ComponentPose=@n162, Alpha=1, Translation="-4.554,-4.672,-5.780", Rotation="23.681,-6.546,4.417") {
    rotationMode=BMM_Additive
    rotationSpace=BCS_BoneSpace
    translationMode=BMM_Additive
    translationSpace=BCS_BoneSpace
    alphaBoolBlend {blendOption=Linear}
    boneToModify {boneName=chain_03_front_03}
    rotation=<rot 23.681 -6.546 4.417>
    translation=<-4.554 -4.672 -5.78>
   }
   n164 = "Transform (Modify) Bone - Bone: chain_03_front_04"(ComponentPose=@n163, Alpha=1, Translation="-0.638,-4.898,-4.578", Rotation="19.549,11.044,4.571") {
    rotationMode=BMM_Additive
    rotationSpace=BCS_BoneSpace
    translationMode=BMM_Additive
    translationSpace=BCS_BoneSpace
    alphaBoolBlend {blendOption=Linear}
    boneToModify {boneName=chain_03_front_04}
    rotation=<rot 19.549 11.0441 4.571>
    translation=<-0.638 -4.898 -4.578>
   }
   n165 = "Transform (Modify) Bone - Bone: chain_03_back_base"(ComponentPose=@n164, Alpha=1, Translation="0,0,0", Rotation="0,0,0") {
    rotationMode=BMM_Additive
    rotationSpace=BCS_BoneSpace
    translationMode=BMM_Additive
    translationSpace=BCS_WorldSpace
    alphaBoolBlend {blendOption=Linear}
    boneToModify {boneName=chain_03_back_base}
   }
   n166 = "Transform (Modify) Bone - Bone: chain_03_back_01"(ComponentPose=@n165, Alpha=1, Translation="0,0,0", Rotation="0,0,0") {
    rotationMode=BMM_Additive
    rotationSpace=BCS_BoneSpace
    translationMode=BMM_Additive
    translationSpace=BCS_BoneSpace
    alphaBoolBlend {blendOption=Linear}
    boneToModify {boneName=chain_03_back_01}
   }
   n167 = "Transform (Modify) Bone - Bone: chain_03_back_02"(ComponentPose=@n166, Alpha=1, Translation="0,0,0", Rotation="0,0,0") {
    rotationMode=BMM_Additive
    rotationSpace=BCS_BoneSpace
    scaleSpace=BCS_BoneSpace
    translationMode=BMM_Additive
    translationSpace=BCS_BoneSpace
    alphaBoolBlend {blendOption=Linear}
    boneToModify {boneName=chain_03_back_02}
   }
   n168 = "Transform (Modify) Bone - Bone: chain_03_back_03"(ComponentPose=@n167, Alpha=1, Translation="-0.191,1.081,0.058", Rotation="-11.728,11.430,-1.445") {
    rotationMode=BMM_Additive
    rotationSpace=BCS_BoneSpace
    scaleSpace=BCS_BoneSpace
    translationMode=BMM_Additive
    translationSpace=BCS_BoneSpace
    alphaBoolBlend {blendOption=Linear}
    boneToModify {boneName=chain_03_back_03}
    rotation=<rot -11.728 11.4301 -1.445>
    translation=<-0.191 1.081 0.058>
   }
   n169 = "Transform (Modify) Bone - Bone: chain_02_front_base"(ComponentPose=@n168, Alpha=1, Translation="0,0,0", Rotation="0.015,0.105,-16.078") {
    rotationMode=BMM_Additive
    rotationSpace=BCS_BoneSpace
    translationMode=BMM_Additive
    translationSpace=BCS_BoneSpace
    alphaBoolBlend {blendOption=Linear}
    boneToModify {boneName=chain_02_front_base}
    rotation=<rot 0.015 0.105 -16.0781>
   }
   n170 = "Transform (Modify) Bone - Bone: chain_02_front_01"(ComponentPose=@n169, Alpha=1, Translation="-0.550,-0.935,3.343", Rotation="19.934,-7.860,-7.092") {
    rotationMode=BMM_Additive
    rotationSpace=BCS_BoneSpace
    translationMode=BMM_Additive
    translationSpace=BCS_BoneSpace
    alphaBoolBlend {blendOption=Linear}
    boneToModify {boneName=chain_02_front_01}
    rotation=<rot 19.934 -7.86 -7.092>
    translation=<-0.55 -0.935 3.343>
   }
   n171 = "Transform (Modify) Bone - Bone: chain_02_front_02"(ComponentPose=@n170, Alpha=1, Translation="0.879,-0.004,6.354", Rotation="4.865,-0.000,0.004") {
    rotationMode=BMM_Additive
    rotationSpace=BCS_BoneSpace
    translationMode=BMM_Additive
    translationSpace=BCS_BoneSpace
    alphaBoolBlend {blendOption=Linear}
    boneToModify {boneName=chain_02_front_02}
    rotation=<rot 4.865 0 0.004>
    translation=<0.879 -0.004 6.354>
   }
   n172 = "Transform (Modify) Bone - Bone: chain_02_front_03"(ComponentPose=@n171, Alpha=1, Translation="-2.336,0.000,1.723", Rotation="-22.370,-0.004,0.030") {
    rotationMode=BMM_Additive
    rotationSpace=BCS_BoneSpace
    translationMode=BMM_Additive
    translationSpace=BCS_BoneSpace
    alphaBoolBlend {blendOption=Linear}
    boneToModify {boneName=chain_02_front_03}
    rotation=<rot -22.37 -0.004 0.03>
    translation {x=-2.336, z=1.723}
   }
   n173 = "Transform (Modify) Bone - Bone: chain_02_front_04"(ComponentPose=@n172, Alpha=1, Translation="-0.534,0.068,-1.486", Rotation="-1.942,-9.694,0.843") {
    rotationMode=BMM_Additive
    rotationSpace=BCS_BoneSpace
    translationMode=BMM_Additive
    translationSpace=BCS_BoneSpace
    alphaBoolBlend {blendOption=Linear}
    boneToModify {boneName=chain_02_front_04}
    rotation=<rot -1.942 -9.694 0.843>
    translation=<-0.534 0.068 -1.486>
   }
   n174 = "Transform (Modify) Bone - Bone: chain_02_back_base"(ComponentPose=@n173, Alpha=1, Translation="0,0,0", Rotation="0,0,0") {
    rotationMode=BMM_Additive
    rotationSpace=BCS_BoneSpace
    translationMode=BMM_Additive
    translationSpace=BCS_BoneSpace
    alphaBoolBlend {blendOption=Linear}
    boneToModify {boneName=chain_02_back_base}
   }
   n175 = "Transform (Modify) Bone - Bone: chain_02_back_01"(ComponentPose=@n174, Alpha=1, Translation="0,0,0", Rotation="0,0,0") {
    rotationMode=BMM_Additive
    rotationSpace=BCS_BoneSpace
    translationMode=BMM_Additive
    translationSpace=BCS_BoneSpace
    alphaBoolBlend {blendOption=Linear}
    boneToModify {boneName=chain_02_back_01}
   }
   n176 = "Transform (Modify) Bone - Bone: chain_02_back_02"(ComponentPose=@n175, Alpha=1, Translation="0,0,0", Rotation="0,0,0") {
    rotationMode=BMM_Additive
    rotationSpace=BCS_BoneSpace
    translationMode=BMM_Additive
    translationSpace=BCS_BoneSpace
    alphaBoolBlend {blendOption=Linear}
    boneToModify {boneName=chain_02_back_02}
   }
   n177 = "Transform (Modify) Bone - Bone: chain_02_back_03"(ComponentPose=@n176, Alpha=1, Translation="0,0,0", Rotation="0.000,10.571,-0.000") {
    rotationMode=BMM_Additive
    rotationSpace=BCS_BoneSpace
    translationMode=BMM_Additive
    translationSpace=BCS_BoneSpace
    alphaBoolBlend {blendOption=Linear}
    boneToModify {boneName=chain_02_back_03}
    rotation=<rot 0 10.571 0>
   }
   n178 = "Transform (Modify) Bone - Bone: chain_01_front_base"(ComponentPose=@n177, Alpha=1, Translation="0,0,0", Rotation="34.242,5.910,3.342") {
    rotationMode=BMM_Additive
    rotationSpace=BCS_BoneSpace
    translationMode=BMM_Additive
    translationSpace=BCS_BoneSpace
    alphaBoolBlend {blendOption=Linear}
    boneToModify {boneName=chain_01_front_base}
    rotation=<rot 34.242 5.91 3.342>
   }
   n179 = "Transform (Modify) Bone - Bone: chain_01_front_01"(ComponentPose=@n178, Alpha=1, Translation="-1.460,0.636,3.874", Rotation="44.301,5.472,14.793") {
    rotationMode=BMM_Additive
    rotationSpace=BCS_BoneSpace
    translationMode=BMM_Additive
    translationSpace=BCS_BoneSpace
    alphaBoolBlend {blendOption=Linear}
    boneToModify {boneName=chain_01_front_01}
    rotation=<rot 44.301 5.472 14.793>
    translation=<-1.46 0.636 3.874>
   }
   n180 = "Transform (Modify) Bone - Bone: chain_01_front_02"(ComponentPose=@n179, Alpha=1, Translation="0.890,1.462,7.488", Rotation="3.936,-0.163,2.989") {
    rotationMode=BMM_Additive
    rotationSpace=BCS_BoneSpace
    translationMode=BMM_Additive
    translationSpace=BCS_BoneSpace
    alphaBoolBlend {blendOption=Linear}
    boneToModify {boneName=chain_01_front_02}
    rotation=<rot 3.936 -0.163 2.989>
    translation=<0.89 1.462 7.488>
   }
   n181 = "Transform (Modify) Bone - Bone: chain_01_front_03"(ComponentPose=@n180, Alpha=1, Translation="-6.209,-0.684,6.378", Rotation="-25.958,24.690,-25.496") {
    rotationMode=BMM_Additive
    rotationSpace=BCS_BoneSpace
    translationMode=BMM_Additive
    translationSpace=BCS_BoneSpace
    alphaBoolBlend {blendOption=Linear}
    boneToModify {boneName=chain_01_front_03}
    rotation=<rot -25.958 24.69 -25.496>
    translation=<-6.209 -0.684 6.378>
   }
   n182 = "Transform (Modify) Bone - Bone: chain_01_front_04"(ComponentPose=@n181, Alpha=1, Translation="-5.243,-2.956,-1.310", Rotation="-20.160,-2.609,0.581") {
    rotationMode=BMM_Additive
    rotationSpace=BCS_BoneSpace
    translationMode=BMM_Additive
    translationSpace=BCS_BoneSpace
    alphaBoolBlend {blendOption=Linear}
    boneToModify {boneName=chain_01_front_04}
    rotation=<rot -20.16 -2.609 0.581>
    translation=<-5.243 -2.956 -1.31>
   }
   n183 = "Transform (Modify) Bone - Bone: chain_01_back_base"(ComponentPose=@n182, Alpha=1, Translation="0,0,0", Rotation="0,0,0") {
    rotationMode=BMM_Additive
    rotationSpace=BCS_BoneSpace
    translationMode=BMM_Additive
    translationSpace=BCS_BoneSpace
    alphaBoolBlend {blendOption=Linear}
    boneToModify {boneName=chain_01_back_base}
   }
   n184 = "Transform (Modify) Bone - Bone: chain_01_back_01"(ComponentPose=@n183, Alpha=1, Translation="0.938,-0.000,-0.000", Rotation="0,0,0") {
    rotationMode=BMM_Additive
    rotationSpace=BCS_BoneSpace
    translationMode=BMM_Additive
    translationSpace=BCS_BoneSpace
    alphaBoolBlend {blendOption=Linear}
    boneToModify {boneName=chain_01_back_01}
    translation {x=0.938}
   }
   n185 = "Transform (Modify) Bone - Bone: chain_01_back_02"(ComponentPose=@n184, Alpha=1, Translation="1.280,-0.659,0.083", Rotation="6.532,5.983,-0.000") {
    rotationMode=BMM_Additive
    rotationSpace=BCS_BoneSpace
    translationMode=BMM_Additive
    translationSpace=BCS_BoneSpace
    alphaBoolBlend {blendOption=Linear}
    boneToModify {boneName=chain_01_back_02}
    rotation=<rot 6.532 5.983 0>
    translation=<1.28 -0.659 0.083>
   }
   n186 = "Transform (Modify) Bone - Bone: chain_01_back_03"(ComponentPose=@n185, Alpha=1, Translation="1.767,1.141,0.170", Rotation="-8.818,26.449,-2.109") {
    rotationMode=BMM_Additive
    rotationSpace=BCS_BoneSpace
    translationMode=BMM_Additive
    translationSpace=BCS_BoneSpace
    alphaBoolBlend {blendOption=Linear}
    boneToModify {boneName=chain_01_back_03}
    rotation=<rot -8.818 26.449 -2.109>
    translation=<1.767 1.141 0.17>
   }
   Save cached pose 'LeftPose'(Pose=Component To Local(ComponentPose=@n186)) {cacheName=LeftPose}
   Reroute Node(InputPin:Wildcard) -> (OutputPin:Wildcard)
  ]}
 EventGraph {nodes [
   Event BlueprintUpdateAnimation()
   > n1 = Sequence!(execute) -> (then_0, then_1, then_2:Exec)
   Set AlphaR_GreaterThen_AlphaL(
    execute=@n1.then_2
    AlphaR_GreaterThen_AlphaL:bool="float >= float"(A=Get Alpha_Thigh_R_Front() -> (Alpha_Thigh_R_Front:double), B=Get Alpha_Thigh_L_Front() -> (Alpha_Thigh_L_Front:double))
   ) -> (Output_Get:bool)
   > Set is not sky diving(is not sky diving:bool=NOT Boolean(A=Get bIsSkyDiving())) -> (Output_Get:bool)
   n2 = Get bIsMoving2D()
   > Set Still_Sitting(Still_Sitting:bool=AND Boolean(A=Get bIsCrouching(), B=NOT Boolean(A=@n2))) -> (Output_Get:bool)
   n3 = Get Skel Mesh()
   Set Alpha_Thigh_R_Front(
    execute=@n1.then_1
    Alpha_Thigh_R_Front:double=GetAlphaFromReferenceTransformDelta(SkeletalMeshComponent=@n3, BoneName=thigh_r, BaseName=pelvis, InputRange=(X=0,Y=-105), TwistAxis="1,0,0")
   ) -> (Output_Get:double)
   > Set Alpha_Thigh_L_Front(Alpha_Thigh_L_Front:double=GetAlphaFromReferenceTransformDelta(SkeletalMeshComponent=@n3, BoneName=thigh_l, BaseName=pelvis, InputRange=(X=0,Y=-105), TwistAxis="1,0,0")) -> (Output_Get:double)
   > Set Alpha_Head_Back(Alpha_Head_Back:double=SelectFloat(
     A=GetAlphaFromReferenceTransformDelta(SkeletalMeshComponent=@n3, BoneName=head, BaseName=neck_02, InputRange=(X=-5,Y=-40), TwistAxis="1,0,0")
     B=GetAlphaFromReferenceTransformDelta(SkeletalMeshComponent=@n3, BoneName=head, BaseName=neck_02, InputRange=(X=-5,Y=-70), TwistAxis="1,0,0")
     bPickA=Get bIsCrouching()
    )) -> (Output_Get:double)
   n4 = GetAlphaFromReferenceTransformDelta(SkeletalMeshComponent=@n3, BoneName=upperarm_l, BaseName=spine_05, DeltaComponentType=EulerY, InputRange=(X=-20,Y=-30), TwistAxis="1,0,0")
   > Set Alpha Head Back Hand Up(Alpha Head Back Hand Up:double=float * float(
     A=Get Alpha_Head_Back() -> (Alpha_Head_Back:double)
     B:bool="In Range (Float)"!(Value:double=@n4.RawOutput, Min:double=10, Max:double=100, InclusiveMin:bool=true, InclusiveMax:bool=true) -> (ReturnValue:bool)
    ) {is=title K2Node_PromotableOperator}) -> (Output_Get:double)
   n5 = GetAlphaFromReferenceTransformDelta(SkeletalMeshComponent=@n3, BoneName=thigh_r, BaseName=pelvis, InputRange=(X=-90,Y=-35), TwistAxis="1,0,0")
   n6 = GetAlphaFromReferenceTransformDelta(SkeletalMeshComponent=@n3, BoneName=thigh_r, BaseName=pelvis, DeltaComponentType=EulerX, InputRange=(X=0,Y=15), TwistAxis="1,0,0")
   > Set Leg Moving Right(Leg Moving Right:float=float * float!(
     A=GetAlphaFromReferenceTransformDelta(SkeletalMeshComponent=@n3, BoneName=thigh_r, BaseName=pelvis, DeltaComponentType=EulerY, InputRange=(X=5,Y=25), TwistAxis="1,0,0")
     B:double=@n5
     C:double=@n6
    ) -> (ReturnValue)) -> (Output_Get:float)
   > Set R leg moving right Peak(R leg moving right Peak:float=float * float(
     A=GetAlphaFromReferenceTransformDelta(SkeletalMeshComponent=@n3, BoneName=thigh_r, BaseName=pelvis, DeltaComponentType=EulerY, InputRange=(X=22,Y=36), TwistAxis="1,0,0")
     B=@n5
     C=@n6
    )) -> (Output_Get:float)
   > Set L Leg going Left(L Leg going Left:float=GetAlphaFromReferenceTransformDelta(SkeletalMeshComponent=@n3, BoneName=thigh_l, BaseName=pelvis, DeltaComponentType=EulerY, InputRange=(X=0,Y=40), TwistAxis="1,0,0")) -> (Output_Get:float)
   n7 = float * float!(
    A=Get Alpha_Thigh_L_Front()
    B=GetAlphaFromReferenceTransformDelta(SkeletalMeshComponent=@n3, BoneName=thigh_l, BaseName=pelvis, DeltaComponentType=EulerY, InputRange=(X=30,Y=-9), TwistAxis="1,0,0")
   ) -> (ReturnValue)
   n8 = Get Alpha_Thigh_R_Front()
   > Set Alpha_Thigh_L_In(Alpha_Thigh_L_In:double=SelectFloat(A=0, B=float - float(A=@n7, B=@n8), bPickA=Get bIsMoving2D())) -> (Output_Get:double)
   > Set Alpha_Thigh_L_Out(Alpha_Thigh_L_Out:double=float - float(
     A=GetAlphaFromReferenceTransformDelta(SkeletalMeshComponent=@n3, BoneName=thigh_l, BaseName=pelvis, DeltaComponentType=EulerY, InputRange=(X=5,Y=30), TwistAxis="1,0,0")
     B=Get Alpha_Thigh_L_Front()
    )) -> (Output_Get:double)
   n9 = GetAlphaFromReferenceTransformDelta(SkeletalMeshComponent=@n3, BoneName=thigh_r, BaseName=pelvis, InputRange=(X=-100,Y=-50), TwistAxis="1,0,0")
   n10 = GetAlphaFromReferenceTransformDelta(SkeletalMeshComponent=@n3, BoneName=thigh_r, BaseName=pelvis, InputRange=(X=0,Y=-50), TwistAxis="1,0,0")
   n11 = "float > float"(A=@n10.RawOutput, B=-50)
   n12 = NOT Boolean(A=Get bIsSkyDiving())
   n13 = OR Boolean(A=Get bIsJumping(), B=Get bIsFalling())
   > Set Alpha_Thigh_R_Jump(Alpha_Thigh_R_Jump:double=SelectFloat(A=SelectFloat(A=@n10, B=@n9, bPickA=@n11), B=0, bPickA=AND Boolean(A=@n13, B=@n12))) -> (Output_Get:double)
   n14 = Get bIsMoving2D()
   > Set Thigh_L_Standing_Straight(Thigh_L_Standing_Straight:double=float * float(
     A=GetAlphaFromReferenceTransformDelta(SkeletalMeshComponent=@n3, BoneName=thigh_l, BaseName=pelvis, InputRange=(X=-15,Y=0), TwistAxis="1,0,0")
     B:bool=NOT Boolean(A=@n14)
    )) -> (Output_Get:double)
   n15 = Gravity Override State Function DEPRECATED(
    execute=@n1.then_0
    CharacterPartAnimInstance=Self-Reference()
    GameTimeInSeconds=GetGameTimeInSeconds()
    GravityOverrideParams=Get Gravity Override Params()
   )
   > Set Gravity Override(Gravity Override=@n15)
   Event BlueprintInitializeAnimation()
   > Set Skel Mesh(Skel Mesh=GetOwnerSkeletalMeshComponent())
   Event OnCharacterCustomizationLoaded()
   > Set Skel Mesh(Skel Mesh=GetPartSkeletalMeshComponent())
  ]}
]
variables [
 Alpha Head Back Hand Up {type=number}
 Alpha_Head_Back {type=number}
 Alpha_Thigh_L_Front {type=number}
 Alpha_Thigh_L_In {type=number}
 Alpha_Thigh_L_Out {type=number}
 Alpha_Thigh_R_Front {type=number}
 Alpha_Thigh_R_Jump {type=number}
 AlphaR_GreaterThen_AlphaL {type=boolean}
 Gravity Override {type=Vector}
 Gravity Override Params {type=GravityOverrideParamsStruct}
 is not sky diving {type=boolean}
 L Leg going Left {type=number}
 Leg Moving Right {type=number}
 R leg moving right Peak {type=number}
 Skel Mesh {type=SkeletalMeshComponent}
 Still_Sitting {type=boolean}
 Thigh_L_Standing_Straight {type=number}
]
